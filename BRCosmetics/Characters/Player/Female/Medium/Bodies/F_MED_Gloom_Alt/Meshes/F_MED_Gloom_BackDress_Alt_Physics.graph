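# Physics asset for the back dress of the Gloom Alt medium female character body, defining bodies and joints for its secondary motion.
# Left and right five-capsule dress chains use tight swing and twist limits, hanging from small master bodies freely jointed to spine_05.
# Colliders cover spine, pelvis, thighs, glutes, short skirt, feet, right calf and right arm, plus two root spheres, one very large and offset downward.
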PhysicsAsset /BRCosmetics/Characters/Player/Female/Medium/Bodies/F_MED_Gloom_Alt/Meshes/F_MED_Gloom_BackDress_Alt_Physics
bodies {
 calf_r [{shapeType=Capsule, length=46.4272, radius=10.114, shapeName=calf_r_capsule, center=<27.6402 0 2.2312>, rotation=<rot 90 -179.1048 167.4784>}]
 deform_glute_base_l [{
   shapeType=Capsule
   length=7.4694
   radius=5.2137
   shapeName=deform_glute_base_l_capsule
   center=<3.7652 -3.4812 -0.6155>
   rotation=<rot -2.8203 1.5657 -7.7614>
  }]
 deform_glute_base_r [{
   shapeType=Capsule
   length=10.5649
   radius=4.5469
   shapeName=deform_glute_base_r_capsule
   center=<-3.6031 4.1216 -0.0739>
   rotation=<rot 0 0 -9.5562>
  }]
 dyn_l_bk_dress_01 [{shapeType=Capsule, length=10.9938, radius=3.6002, shapeName=dyn_l_bk_dress_01_capsule, center=<5.4995 0 3>, rotation=<rot 90 0 0>}]
 dyn_l_bk_dress_02 [{shapeType=Capsule, length=8.7598, radius=5.8342, shapeName=dyn_l_bk_dress_02_capsule, center=<5.4995 0 3>, rotation=<rot 90 0 0>}]
 dyn_l_bk_dress_03 [{shapeType=Capsule, length=8.702, radius=5.892, shapeName=dyn_l_bk_dress_03_capsule, center=<5.4995 0 3>, rotation=<rot 90 0 0>}]
 dyn_l_bk_dress_04 [{shapeType=Capsule, length=3.7505, radius=10.5219, shapeName=dyn_l_bk_dress_04_capsule, center=<5.4995 0 3>, rotation=<rot 90 0 0>}]
 dyn_l_bk_dress_05 [{shapeType=Capsule, length=8.3805, radius=5.892, shapeName=dyn_l_bk_dress_05_capsule, center=<5.4995 0 3>, rotation=<rot 90 0 0>}]
 dyn_l_bk_dress_MSTR [{shapeType=Capsule, length=0.505, radius=0.505, shapeName=dyn_l_bk_dress_MSTR_capsule, rotation=<rot 0 -90 0>}]
 dyn_l_fr_dress_MSTR [{shapeType=Capsule, length=0.505, radius=0.505, shapeName=dyn_l_fr_dress_MSTR_capsule, rotation=<rot 0 -90 0>}]
 dyn_r_bk_dress_01 [{shapeType=Capsule, length=10.9938, radius=3.6002, shapeName=dyn_r_bk_dress_01_capsule, center=<5.4995 0 3>, rotation=<rot 90 0 0>}]
 dyn_r_bk_dress_02 [{shapeType=Capsule, length=8.7598, radius=5.8342, shapeName=dyn_r_bk_dress_02_capsule, center=<5.4995 0 3>, rotation=<rot 90 0 0>}]
 dyn_r_bk_dress_03 [{shapeType=Capsule, length=1.7948, radius=10, shapeName=dyn_r_bk_dress_03_capsule, center=<5.4995 0 3>, rotation=<rot 90 0 0>}]
 dyn_r_bk_dress_04 [{shapeType=Capsule, length=4.6445, radius=9.628, shapeName=dyn_r_bk_dress_04_capsule, center=<5.4995 0 3>, rotation=<rot 90 0 0>}]
 dyn_r_bk_dress_05 [{shapeType=Capsule, length=8.3805, radius=5.892, shapeName=dyn_r_bk_dress_05_capsule, center=<5.4995 0 3>, rotation=<rot 90 0 0>}]
 dyn_r_bk_dress_MSTR [{shapeType=Capsule, length=0.505, radius=0.505, shapeName=dyn_r_bk_dress_MSTR_capsule, rotation=<rot 0 -90 0>}]
 dyn_r_fr_dress_MSTR [{shapeType=Capsule, length=0.505, radius=0.505, shapeName=dyn_r_fr_dress_MSTR_capsule, rotation=<rot 0 -90 0>}]
 dyn_shortSkirt [{
   shapeType=Capsule
   length=12.7416
   radius=5.872
   shapeName=dyn_shortSkirt_capsule
   center=<3.1877 1.667 -1.8745>
   rotation=<rot 0 -8.841 -84.8001>
  }]
 foot_l [{shapeType=Capsule, length=9.3802, radius=8.2614, shapeName=foot_l_capsule, rotation=<rot 0 0 90>}]
 foot_r [{shapeType=Capsule, length=9.2833, radius=7.3326, shapeName=foot_r_capsule, rotation=<rot 0 0 90>}]
 lowerarm_r [{shapeType=Capsule, length=25.2808, radius=5.834, shapeName=upperarm_r_capsule, center=<-14.3308 0 0>, rotation=<rot 90 0 0>}]
 pelvis [{shapeType=Capsule, length=133.6039, radius=10.4707}]
 root [
  {shapeType=Sphere, radius=195.2559, shapeName=root_sphere, center=<0 -8.6403 -175.5073>}
  {shapeType=Sphere, radius=27.8343, shapeName=root_sphere, center=<0.0002 2.4787 8.5972>}
 ]
 spine_01 [{shapeType=Capsule, length=133.6039, radius=10.4707}]
 spine_02 [{shapeType=Capsule, length=136.9789, radius=7.0957, center=<0 -1.272 0>}]
 spine_03 [{shapeType=Capsule, length=136.2388, radius=7.8358, center=<0 -1.272 0>}]
 spine_04 [{shapeType=Capsule, length=138.5418, radius=5.5328, center=<0 -1.272 0>}]
 spine_05 [{shapeType=Capsule, length=111.158, radius=30.4443, center=<3.6444 22.6364 0>}]
 thigh_l [
  {shapeType=Capsule, length=32.8459, radius=7.7646, center=<-20.6473 0 0>, rotation=<rot 90 0 0>}
  {shapeType=Capsule, length=34.0816, radius=6.5289, shapeName=thigh_l_capsule, center=<-20.7665 0 -10.7364>, rotation=<rot 90 0 0>}
 ]
 thigh_r [
  {shapeType=Capsule, length=32.8459, radius=7.7646, center=<20.4782 0 0>, rotation=<rot 90 0 0>}
  {shapeType=Capsule, length=34.0816, radius=6.5289, shapeName=thigh_r_capsule, center=<20.217 0 9.5963>, rotation=<rot 90 0 0>}
 ]
 upperarm_r [{shapeType=Capsule, length=25.7519, radius=5.363, shapeName=upperarm_r_capsule, center=<-13.9177 0 0>, rotation=<rot 90 0 0>}]
}
constraints [
 {bone1Name=spine_02, bone2Name=spine_01, swing1LimitDegrees=45, swing2LimitDegrees=45, twistLimitDegrees=45}
 {bone1Name=spine_04, bone2Name=spine_03, swing1LimitDegrees=45, swing2LimitDegrees=45, twistLimitDegrees=45}
 {bone1Name=spine_05, bone2Name=spine_04, swing1LimitDegrees=45, swing2LimitDegrees=45, twistLimitDegrees=45}
 {bone1Name=thigh_l, bone2Name=pelvis, swing1LimitDegrees=45, swing2LimitDegrees=45, twistLimitDegrees=45}
 {bone1Name=thigh_r, bone2Name=pelvis, swing1LimitDegrees=45, swing2LimitDegrees=45, twistLimitDegrees=45}
 {
  bone1Name=spine_01
  bone2Name=pelvis
  swing1LimitDegrees=45
  swing1Motion=Free
  swing2LimitDegrees=45
  swing2Motion=Free
  twistLimitDegrees=45
  twistMotion=Free
 }
 {
  bone1Name=dyn_l_fr_dress_MSTR
  bone2Name=spine_05
  swing1LimitDegrees=45
  swing1Motion=Free
  swing2LimitDegrees=45
  swing2Motion=Free
  twistLimitDegrees=45
  twistMotion=Free
 }
 {
  bone1Name=dyn_l_bk_dress_MSTR
  bone2Name=spine_05
  swing1LimitDegrees=45
  swing1Motion=Free
  swing2LimitDegrees=45
  swing2Motion=Free
  twistLimitDegrees=45
  twistMotion=Free
 }
 {
  bone1Name=dyn_r_fr_dress_MSTR
  bone2Name=spine_05
  swing1LimitDegrees=45
  swing1Motion=Free
  swing2LimitDegrees=45
  swing2Motion=Free
  twistLimitDegrees=45
  twistMotion=Free
 }
 {
  bone1Name=dyn_r_bk_dress_MSTR
  bone2Name=spine_05
  swing1LimitDegrees=45
  swing1Motion=Free
  swing2LimitDegrees=45
  swing2Motion=Free
  twistLimitDegrees=45
  twistMotion=Free
 }
 {
  bone1Name=spine_03
  bone2Name=spine_02
  swing1LimitDegrees=45
  swing1Motion=Free
  swing2LimitDegrees=45
  swing2Motion=Free
  twistLimitDegrees=45
  twistMotion=Free
 }
 {bone1Name=dyn_r_bk_dress_01, bone2Name=dyn_r_bk_dress_MSTR, swing1LimitDegrees=10, swing2LimitDegrees=20, twistLimitDegrees=10}
 {bone1Name=dyn_r_bk_dress_02, bone2Name=dyn_r_bk_dress_01, swing1LimitDegrees=10, swing2LimitDegrees=20, twistLimitDegrees=10}
 {bone1Name=dyn_r_bk_dress_03, bone2Name=dyn_r_bk_dress_02, swing1LimitDegrees=10, swing2LimitDegrees=20, twistLimitDegrees=10}
 {bone1Name=dyn_r_bk_dress_04, bone2Name=dyn_r_bk_dress_03, swing1LimitDegrees=10, swing2LimitDegrees=20, twistLimitDegrees=10}
 {bone1Name=dyn_r_bk_dress_05, bone2Name=dyn_r_bk_dress_04, swing1LimitDegrees=10, swing2LimitDegrees=20, twistLimitDegrees=10}
 {bone1Name=dyn_l_bk_dress_01, bone2Name=dyn_l_bk_dress_MSTR, swing1LimitDegrees=10, swing2LimitDegrees=20, twistLimitDegrees=10}
 {bone1Name=dyn_l_bk_dress_02, bone2Name=dyn_l_bk_dress_01, swing1LimitDegrees=10, swing2LimitDegrees=20, twistLimitDegrees=10}
 {bone1Name=dyn_l_bk_dress_03, bone2Name=dyn_l_bk_dress_02, swing1LimitDegrees=10, swing2LimitDegrees=20, twistLimitDegrees=10}
 {bone1Name=dyn_l_bk_dress_04, bone2Name=dyn_l_bk_dress_03, swing1LimitDegrees=10, swing2LimitDegrees=20, twistLimitDegrees=10}
 {bone1Name=dyn_l_bk_dress_05, bone2Name=dyn_l_bk_dress_04, swing1LimitDegrees=10, swing2LimitDegrees=20, twistLimitDegrees=10}
 {
  bone1Name=upperarm_r
  bone2Name=spine_05
  swing1LimitDegrees=45
  swing1Motion=Free
  swing2LimitDegrees=45
  swing2Motion=Free
  twistLimitDegrees=45
  twistMotion=Free
 }
 {
  bone1Name=deform_glute_base_l
  bone2Name=thigh_l
  swing1LimitDegrees=45
  swing1Motion=Free
  swing2LimitDegrees=45
  swing2Motion=Free
  twistLimitDegrees=45
  twistMotion=Free
 }
 {
  bone1Name=deform_glute_base_r
  bone2Name=thigh_r
  swing1LimitDegrees=45
  swing1Motion=Free
  swing2LimitDegrees=45
  swing2Motion=Free
  twistLimitDegrees=45
  twistMotion=Free
 }
 {
  bone1Name=dyn_shortSkirt
  bone2Name=thigh_l
  swing1LimitDegrees=45
  swing1Motion=Free
  swing2LimitDegrees=45
  swing2Motion=Free
  twistLimitDegrees=45
  twistMotion=Free
 }
 {
  bone1Name=calf_r
  bone2Name=thigh_r
  swing1LimitDegrees=45
  swing1Motion=Free
  swing2LimitDegrees=45
  swing2Motion=Free
  twistLimitDegrees=45
  twistMotion=Free
 }
 {
  bone1Name=foot_l
  bone2Name=thigh_l
  swing1LimitDegrees=45
  swing1Motion=Free
  swing2LimitDegrees=45
  swing2Motion=Free
  twistLimitDegrees=45
  twistMotion=Free
 }
 {
  bone1Name=foot_r
  bone2Name=calf_r
  swing1LimitDegrees=45
  swing1Motion=Free
  swing2LimitDegrees=45
  swing2Motion=Free
  twistLimitDegrees=45
  twistMotion=Free
 }
 {
  bone1Name=pelvis
  bone2Name=root
  swing1LimitDegrees=45
  swing1Motion=Free
  swing2LimitDegrees=45
  swing2Motion=Free
  twistLimitDegrees=45
  twistMotion=Free
 }
 {
  bone1Name=lowerarm_r
  bone2Name=upperarm_r
  swing1LimitDegrees=45
  swing1Motion=Free
  swing2LimitDegrees=45
  swing2Motion=Free
  twistLimitDegrees=45
  twistMotion=Free
 }
]
