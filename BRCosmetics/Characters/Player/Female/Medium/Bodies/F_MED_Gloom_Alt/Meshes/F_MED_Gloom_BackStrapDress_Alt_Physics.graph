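# Physics asset for the back dress straps on the alternate Gloom medium female player body.
# Left and right straps are three-capsule chains with swing and twist limits, each hanging from a tiny master body constrained to the pelvis.
# Capsules on the pelvis, spine, thighs and right upper arm approximate the body.
# Four front and back dress master bodies attach to the upper spine with free rotation but have no child chains here.
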PhysicsAsset /BRCosmetics/Characters/Player/Female/Medium/Bodies/F_MED_Gloom_Alt/Meshes/F_MED_Gloom_BackStrapDress_Alt_Physics
bodies {
 dyn_l_bk_dressStrap_01 [{shapeType=Capsule, length=6.9197, radius=3.0803, shapeName=dyn_l_bk_dressStrap_01_capsule, center=<10.7644 0 2>, rotation=<rot 90 0 0>}]
 dyn_l_bk_dressStrap_02 [{shapeType=Capsule, length=8.0106, radius=1.9894, shapeName=dyn_l_bk_dressStrap_02_capsule, center=<5.1438 0 0>, rotation=<rot 90 0 0>}]
 dyn_l_bk_dressStrap_03 [{shapeType=Capsule, length=8.0106, radius=1.9894, shapeName=dyn_l_bk_dressStrap_03_capsule, center=<5.1438 0 0>, rotation=<rot 90 0 0>}]
 dyn_l_bk_dressStrap_MSTR [{shapeType=Capsule, length=0.505, radius=0.505, shapeName=dyn_l_bk_dressStrap_MSTR_capsule, rotation=<rot 0 -90 0>}]
 dyn_l_bk_dress_MSTR [{shapeType=Capsule, length=0.505, radius=0.505, shapeName=dyn_l_bk_dress_MSTR_capsule, rotation=<rot 0 -90 0>}]
 dyn_l_fr_dress_MSTR [{shapeType=Capsule, length=0.505, radius=0.505, shapeName=dyn_l_fr_dress_MSTR_capsule, rotation=<rot 0 -90 0>}]
 dyn_r_bk_dressStrap_01 [{shapeType=Capsule, length=7.7694, radius=2.2306, shapeName=dyn_r_bk_dressStrap_01_capsule, center=<10.7644 0 2>, rotation=<rot 90 0 0>}]
 dyn_r_bk_dressStrap_02 [{shapeType=Capsule, length=7.2033, radius=2.7967, shapeName=dyn_r_bk_dressStrap_02_capsule, center=<5.1438 0 0>, rotation=<rot 90 0 0>}]
 dyn_r_bk_dressStrap_03 [{shapeType=Capsule, length=8.2693, radius=1.7307, shapeName=dyn_r_bk_dressStrap_03_capsule, center=<5.1438 0 0>, rotation=<rot 90 0 0>}]
 dyn_r_bk_dressStrap_MSTR [{shapeType=Capsule, length=0.505, radius=0.505, shapeName=dyn_r_bk_dressStrap_MSTR_capsule, rotation=<rot 0 -90 0>}]
 dyn_r_bk_dress_MSTR [{shapeType=Capsule, length=0.505, radius=0.505, shapeName=dyn_r_bk_dress_MSTR_capsule, rotation=<rot 0 -90 0>}]
 dyn_r_fr_dress_MSTR [{shapeType=Capsule, length=0.505, radius=0.505, shapeName=dyn_r_fr_dress_MSTR_capsule, rotation=<rot 0 -90 0>}]
 pelvis [{shapeType=Capsule, length=133.6039, radius=10.4707}]
 spine_01 [{shapeType=Capsule, length=133.6039, radius=10.4707}]
 spine_02 [{shapeType=Capsule, length=136.9789, radius=7.0957, center=<0 -1.272 0>}]
 spine_03 [{shapeType=Capsule, length=136.2388, radius=7.8358, center=<0 -1.272 0>}]
 spine_04 [{shapeType=Capsule, length=138.5418, radius=5.5328, center=<0 -1.272 0>}]
 spine_05 [{shapeType=Capsule, length=114.4156, radius=27.1866, center=<3.6444 22.6364 0>}]
 thigh_l [
  {shapeType=Capsule, length=32.8459, radius=7.7646, center=<-20.6473 0 0>, rotation=<rot 90 0 0>}
  {shapeType=Capsule, length=34.0816, radius=6.5289, shapeName=thigh_l_capsule, center=<-20.7665 0 -10.7364>, rotation=<rot 90 0 0>}
 ]
 thigh_r [
  {shapeType=Capsule, length=32.8459, radius=7.7646, center=<20.4782 0 0>, rotation=<rot 90 0 0>}
  {shapeType=Capsule, length=34.0816, radius=6.5289, shapeName=thigh_r_capsule, center=<20.217 0 9.5963>, rotation=<rot 90 0 0>}
 ]
 upperarm_r [{shapeType=Capsule, length=18.2453, radius=6.1778, shapeName=upperarm_r_capsule, center=<-10.6844 0 0>, rotation=<rot 90 0 0>}]
}
constraints [
 {bone1Name=spine_02, bone2Name=spine_01, swing1LimitDegrees=45, swing2LimitDegrees=45, twistLimitDegrees=45}
 {bone1Name=spine_04, bone2Name=spine_03, swing1LimitDegrees=45, swing2LimitDegrees=45, twistLimitDegrees=45}
 {bone1Name=spine_05, bone2Name=spine_04, swing1LimitDegrees=45, swing2LimitDegrees=45, twistLimitDegrees=45}
 {bone1Name=thigh_l, bone2Name=pelvis, swing1LimitDegrees=45, swing2LimitDegrees=45, twistLimitDegrees=45}
 {bone1Name=thigh_r, bone2Name=pelvis, swing1LimitDegrees=45, swing2LimitDegrees=45, twistLimitDegrees=45}
 {
  bone1Name=spine_01
  bone2Name=pelvis
  swing1LimitDegrees=45
  swing1Motion=Free
  swing2LimitDegrees=45
  swing2Motion=Free
  twistLimitDegrees=45
  twistMotion=Free
 }
 {
  bone1Name=dyn_l_fr_dress_MSTR
  bone2Name=spine_05
  swing1LimitDegrees=45
  swing1Motion=Free
  swing2LimitDegrees=45
  swing2Motion=Free
  twistLimitDegrees=45
  twistMotion=Free
 }
 {
  bone1Name=dyn_l_bk_dress_MSTR
  bone2Name=spine_05
  swing1LimitDegrees=45
  swing1Motion=Free
  swing2LimitDegrees=45
  swing2Motion=Free
  twistLimitDegrees=45
  twistMotion=Free
 }
 {
  bone1Name=dyn_r_fr_dress_MSTR
  bone2Name=spine_05
  swing1LimitDegrees=45
  swing1Motion=Free
  swing2LimitDegrees=45
  swing2Motion=Free
  twistLimitDegrees=45
  twistMotion=Free
 }
 {
  bone1Name=dyn_r_bk_dress_MSTR
  bone2Name=spine_05
  swing1LimitDegrees=45
  swing1Motion=Free
  swing2LimitDegrees=45
  swing2Motion=Free
  twistLimitDegrees=45
  twistMotion=Free
 }
 {
  bone1Name=dyn_l_bk_dressStrap_MSTR
  bone2Name=pelvis
  swing1LimitDegrees=45
  swing1Motion=Free
  swing2LimitDegrees=45
  swing2Motion=Free
  twistLimitDegrees=45
  twistMotion=Free
 }
 {
  bone1Name=dyn_r_bk_dressStrap_MSTR
  bone2Name=pelvis
  swing1LimitDegrees=45
  swing1Motion=Free
  swing2LimitDegrees=45
  swing2Motion=Free
  twistLimitDegrees=45
  twistMotion=Free
 }
 {
  bone1Name=spine_03
  bone2Name=spine_02
  swing1LimitDegrees=45
  swing1Motion=Free
  swing2LimitDegrees=45
  swing2Motion=Free
  twistLimitDegrees=45
  twistMotion=Free
 }
 {bone1Name=dyn_l_bk_dressStrap_01, bone2Name=dyn_l_bk_dressStrap_MSTR, swing1LimitDegrees=50, swing2LimitDegrees=50, twistLimitDegrees=30}
 {bone1Name=dyn_l_bk_dressStrap_02, bone2Name=dyn_l_bk_dressStrap_01, swing1LimitDegrees=50, swing2LimitDegrees=50, twistLimitDegrees=30}
 {bone1Name=dyn_l_bk_dressStrap_03, bone2Name=dyn_l_bk_dressStrap_02, swing1LimitDegrees=50, swing2LimitDegrees=50, twistLimitDegrees=30}
 {bone1Name=dyn_r_bk_dressStrap_01, bone2Name=dyn_r_bk_dressStrap_MSTR, swing1LimitDegrees=50, swing2LimitDegrees=50, twistLimitDegrees=30}
 {bone1Name=dyn_r_bk_dressStrap_02, bone2Name=dyn_r_bk_dressStrap_01, swing1LimitDegrees=50, swing2LimitDegrees=50, twistLimitDegrees=30}
 {bone1Name=dyn_r_bk_dressStrap_03, bone2Name=dyn_r_bk_dressStrap_02, swing1LimitDegrees=50, swing2LimitDegrees=50, twistLimitDegrees=30}
 {
  bone1Name=upperarm_r
  bone2Name=spine_05
  swing1LimitDegrees=45
  swing1Motion=Free
  swing2LimitDegrees=45
  swing2Motion=Free
  twistLimitDegrees=45
  twistMotion=Free
 }
]
