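# Physics asset for the front dress of the female medium Gloom Alt character, named as the emote variant.
# Capsules on spine, pelvis, thighs and calves form the body shape, with two root spheres, one very large and offset from the bone.
# Left and right front-dress bone capsules constrain directly to the pelvis, with narrow first swing, wide second swing and zero-degree twist limit.
# Body joints mix free and limited angular motion.
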PhysicsAsset /BRCosmetics/Characters/Player/Female/Medium/Bodies/F_MED_Gloom_Alt/Meshes/F_MED_Gloom_FrontDress_Alt_Emote_Physics
bodies {
 calf_l [{shapeType=Capsule, length=46.2293, radius=8.4634, center=<-25.0831 -0.0636 0>, rotation=<rot 90 0 0>}]
 calf_r [{shapeType=Capsule, length=41.8862, radius=7.0873, center=<30.9747 2.062 0>, rotation=<rot 90 0 0>}]
 dyn_l_fr_dress_03 [{shapeType=Capsule, length=8.25, radius=5.75, shapeName=dyn_l_fr_dress_03_capsule, center=<9.4718 0 1.4953>, rotation=<rot 90 0 0>}]
 dyn_r_fr_dress_03 [{shapeType=Capsule, length=8.25, radius=5.75, shapeName=dyn_r_fr_dress_03_capsule, center=<9.4718 0 1.4953>, rotation=<rot 90 0 0>}]
 pelvis [{shapeType=Capsule, length=17.9445, radius=11.8173, center=<0 -1.2791 0>}]
 root [
  {shapeType=Sphere, radius=183.4375, shapeName=root_sphere, center=<0 -8.6403 -175.5073>}
  {shapeType=Sphere, radius=27.8343, shapeName=root_sphere, center=<0.0002 2.4787 8.5972>}
 ]
 spine_01 [{shapeType=Capsule, length=70.505, radius=5.755, shapeName=spine_01_capsule, rotation=<rot 0 -90 0>}]
 spine_02 [{shapeType=Capsule, length=66.391, radius=7.8582, shapeName=spine_02_capsule}]
 spine_03 [{shapeType=Capsule, length=66.005, radius=7.005, shapeName=spine_03_capsule}]
 spine_04 [{shapeType=Capsule, length=49.9289, radius=6.7805, center=<0 0.3471 0>}]
 spine_05 [{shapeType=Capsule, length=31.4921, radius=9.4395}]
 thigh_l [
  {shapeType=Capsule, length=30.2293, radius=8.4634, center=<-26.5449 -0.6527 0>, rotation=<rot 90 0 0>}
  {
   shapeType=Capsule
   length=39.3132
   radius=7.7855
   shapeName=thigh_l_capsule
   center=<-15.1021 0 -6.8696>
   rotation=<rot 90 -175.2364 -85.2362>
  }
 ]
 thigh_r [
  {shapeType=Capsule, length=30.3862, radius=7.5873, center=<26.7239 2.0804 0>, rotation=<rot 90 0 0>}
  {shapeType=Capsule, length=39.2311, radius=7.8677, shapeName=thigh_r_capsule, center=<15.9361 0 6.8683>, rotation=<rot 90 -165.9638 -75.9636>}
 ]
}
constraints [
 {bone1Name=spine_04, bone2Name=spine_03, swing1LimitDegrees=45, swing2LimitDegrees=45, twistLimitDegrees=45}
 {bone1Name=spine_05, bone2Name=spine_04, swing1LimitDegrees=45, swing2LimitDegrees=45, twistLimitDegrees=45}
 {bone1Name=thigh_l, bone2Name=pelvis, swing1LimitDegrees=45, swing2LimitDegrees=45, twistLimitDegrees=45}
 {bone1Name=thigh_r, bone2Name=pelvis, swing1LimitDegrees=45, swing2LimitDegrees=45, twistLimitDegrees=45}
 {
  bone1Name=pelvis
  bone2Name=root
  swing1LimitDegrees=45
  swing1Motion=Free
  swing2LimitDegrees=45
  swing2Motion=Free
  twistLimitDegrees=45
  twistMotion=Free
 }
 {
  bone1Name=calf_l
  bone2Name=thigh_l
  swing1LimitDegrees=45
  swing1Motion=Free
  swing2LimitDegrees=45
  swing2Motion=Free
  twistLimitDegrees=45
  twistMotion=Free
 }
 {
  bone1Name=calf_r
  bone2Name=thigh_r
  swing1LimitDegrees=45
  swing1Motion=Free
  swing2LimitDegrees=45
  swing2Motion=Free
  twistLimitDegrees=45
  twistMotion=Free
 }
 {
  bone1Name=spine_03
  bone2Name=spine_02
  swing1LimitDegrees=45
  swing1Motion=Free
  swing2LimitDegrees=45
  swing2Motion=Free
  twistLimitDegrees=45
  twistMotion=Free
 }
 {
  bone1Name=spine_02
  bone2Name=spine_01
  swing1LimitDegrees=45
  swing1Motion=Free
  swing2LimitDegrees=45
  swing2Motion=Free
  twistLimitDegrees=45
  twistMotion=Free
 }
 {bone1Name=dyn_l_fr_dress_03, bone2Name=pelvis, swing1LimitDegrees=5, swing2LimitDegrees=120, twistLimitDegrees=0}
 {bone1Name=dyn_r_fr_dress_03, bone2Name=pelvis, swing1LimitDegrees=5, swing2LimitDegrees=120, twistLimitDegrees=0}
 {
  bone1Name=spine_01
  bone2Name=pelvis
  swing1LimitDegrees=45
  swing1Motion=Free
  swing2LimitDegrees=45
  swing2Motion=Free
  twistLimitDegrees=45
  twistMotion=Free
 }
]
