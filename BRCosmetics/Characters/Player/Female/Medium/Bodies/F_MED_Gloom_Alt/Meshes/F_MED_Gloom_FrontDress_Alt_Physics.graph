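# Physics asset for the front dress of the alternate F_MED_Gloom player cosmetic body, defining the shapes and joints for its dress bones.
# Left and right chains of five capsule bodies hang from small MSTR bodies, which link to spine_05 with free rotation.
# Chain joints favor swing2, with tighter swing1 and twist limits that loosen on the last two segments.
# Pelvis, spine and multi-capsule thigh shapes, plus two root spheres, one very large and offset, approximate the body.
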PhysicsAsset /BRCosmetics/Characters/Player/Female/Medium/Bodies/F_MED_Gloom_Alt/Meshes/F_MED_Gloom_FrontDress_Alt_Physics
bodies {
 dyn_l_fr_dress_01 [{shapeType=Capsule, length=2.4017, radius=6.4967, shapeName=dyn_l_fr_dress_01_capsule, center=<9.4667 0 6.3322>, rotation=<rot 90 0 0>}]
 dyn_l_fr_dress_02 [{shapeType=Capsule, length=2.4017, radius=6.4967, shapeName=dyn_l_fr_dress_02_capsule, center=<9.4667 0 6.3322>, rotation=<rot 90 0 0>}]
 dyn_l_fr_dress_03 [{shapeType=Capsule, length=18.6283, radius=7.6832, shapeName=dyn_l_fr_dress_03_capsule, center=<9.4667 0 6.3322>, rotation=<rot 90 0 0>}]
 dyn_l_fr_dress_04 [{shapeType=Capsule, length=21.6175, radius=8.0783, shapeName=dyn_l_fr_dress_04_capsule, center=<9.4667 0 6.3322>, rotation=<rot 90 0 0>}]
 dyn_l_fr_dress_05 [{shapeType=Capsule, length=6.8527, radius=8.0783, shapeName=dyn_l_fr_dress_05_capsule, center=<9.4667 0 6.3322>, rotation=<rot 90 0 0>}]
 dyn_l_fr_dress_MSTR [{shapeType=Capsule, length=0.505, radius=0.505, shapeName=dyn_l_fr_dress_MSTR_capsule, rotation=<rot 0 -90 0>}]
 dyn_r_fr_dress_01 [{shapeType=Capsule, length=0.7678, radius=8.1306, shapeName=dyn_r_fr_dress_01_capsule, center=<9.4667 0 6.3322>, rotation=<rot 90 0 0>}]
 dyn_r_fr_dress_02 [{shapeType=Capsule, length=0.7678, radius=8.1306, shapeName=dyn_r_fr_dress_02_capsule, center=<9.4667 0 6.3322>, rotation=<rot 90 0 0>}]
 dyn_r_fr_dress_03 [{shapeType=Capsule, length=6.1945, radius=8.7365, shapeName=dyn_r_fr_dress_03_capsule, center=<9.4667 0 6.3322>, rotation=<rot 90 0 0>}]
 dyn_r_fr_dress_04 [{shapeType=Capsule, length=5.7994, radius=9.1316, shapeName=dyn_r_fr_dress_04_capsule, center=<9.4667 0 6.3322>, rotation=<rot 90 0 0>}]
 dyn_r_fr_dress_05 [{shapeType=Capsule, length=5.7994, radius=9.1316, shapeName=dyn_r_fr_dress_05_capsule, center=<9.4667 0 6.3322>, rotation=<rot 90 0 0>}]
 dyn_r_fr_dress_MSTR [{shapeType=Capsule, length=0.505, radius=0.505, shapeName=dyn_r_fr_dress_MSTR_capsule, rotation=<rot 0 -90 0>}]
 pelvis [{shapeType=Capsule, length=131.8578, radius=7.0303, center=<0 -1.2791 0>}]
 root [
  {shapeType=Sphere, radius=183.4375, shapeName=root_sphere, center=<0 -8.6403 -175.5073>}
  {shapeType=Sphere, radius=27.8343, shapeName=root_sphere, center=<0.0002 2.4787 8.5972>}
 ]
 spine_01 [{shapeType=Capsule, length=135.5943, radius=6.0086}]
 spine_02 [{shapeType=Capsule, length=136.4493, radius=5.1536, center=<0 0.9981 0>}]
 spine_03 [{shapeType=Capsule, length=108.995, radius=32.6073, center=<3.6444 -27.5509 0>}]
 spine_04 [{shapeType=Capsule, length=8.6789, radius=6.7805, center=<0 5.5947 0>}]
 spine_05 [{shapeType=Capsule, length=16.4921, radius=9.4395}]
 thigh_l [
  {shapeType=Capsule, length=38.5676, radius=10.9415, center=<-26.2393 0 0>, rotation=<rot 90 0 0>}
  {shapeType=Capsule, length=37.6145, radius=12.2342, shapeName=thigh_l_capsule, center=<-28.6866 0 -17.6251>, rotation=<rot 90 0 0>}
  {
   shapeType=Capsule
   length=14.3298
   radius=11.4608
   shapeName=thigh_l_capsule
   center=<-11.8475 3.1462 4.8365>
   rotation=<rot 48.3474 180 180>
  }
 ]
 thigh_r [
  {shapeType=Capsule, length=31.5689, radius=17.4046, center=<25.6654 9.6906 0>, rotation=<rot 90 0 0>}
  {shapeType=Capsule, length=39.4811, radius=10.3677, shapeName=thigh_r_capsule, center=<24.4354 0 19.3126>, rotation=<rot 90 0 0>}
 ]
}
constraints [
 {bone1Name=spine_02, bone2Name=spine_01, swing1LimitDegrees=45, swing2LimitDegrees=45, twistLimitDegrees=45}
 {bone1Name=spine_04, bone2Name=spine_03, swing1LimitDegrees=45, swing2LimitDegrees=45, twistLimitDegrees=45}
 {bone1Name=spine_05, bone2Name=spine_04, swing1LimitDegrees=45, swing2LimitDegrees=45, twistLimitDegrees=45}
 {bone1Name=thigh_l, bone2Name=pelvis, swing1LimitDegrees=45, swing2LimitDegrees=45, twistLimitDegrees=45}
 {bone1Name=thigh_r, bone2Name=pelvis, swing1LimitDegrees=45, swing2LimitDegrees=45, twistLimitDegrees=45}
 {
  bone1Name=spine_01
  bone2Name=pelvis
  swing1LimitDegrees=45
  swing1Motion=Free
  swing2LimitDegrees=45
  swing2Motion=Free
  twistLimitDegrees=45
  twistMotion=Free
 }
 {
  bone1Name=dyn_l_fr_dress_MSTR
  bone2Name=spine_05
  swing1LimitDegrees=45
  swing1Motion=Free
  swing2LimitDegrees=45
  swing2Motion=Free
  twistLimitDegrees=45
  twistMotion=Free
 }
 {
  bone1Name=dyn_r_fr_dress_MSTR
  bone2Name=spine_05
  swing1LimitDegrees=45
  swing1Motion=Free
  swing2LimitDegrees=45
  swing2Motion=Free
  twistLimitDegrees=45
  twistMotion=Free
 }
 {bone1Name=dyn_l_fr_dress_01, bone2Name=dyn_l_fr_dress_MSTR, swing1LimitDegrees=5, swing2LimitDegrees=30, twistLimitDegrees=5}
 {bone1Name=dyn_l_fr_dress_02, bone2Name=dyn_l_fr_dress_01, swing1LimitDegrees=5, swing2LimitDegrees=30, twistLimitDegrees=5}
 {bone1Name=dyn_l_fr_dress_03, bone2Name=dyn_l_fr_dress_02, swing1LimitDegrees=5, swing2LimitDegrees=30, twistLimitDegrees=5}
 {bone1Name=dyn_l_fr_dress_04, bone2Name=dyn_l_fr_dress_03, swing1LimitDegrees=15, swing2LimitDegrees=30, twistLimitDegrees=15}
 {bone1Name=dyn_l_fr_dress_05, bone2Name=dyn_l_fr_dress_04, swing1LimitDegrees=15, swing2LimitDegrees=30, twistLimitDegrees=15}
 {bone1Name=dyn_r_fr_dress_01, bone2Name=dyn_r_fr_dress_MSTR, swing1LimitDegrees=5, swing2LimitDegrees=30, twistLimitDegrees=5}
 {bone1Name=dyn_r_fr_dress_02, bone2Name=dyn_r_fr_dress_01, swing1LimitDegrees=5, swing2LimitDegrees=30, twistLimitDegrees=5}
 {bone1Name=dyn_r_fr_dress_03, bone2Name=dyn_r_fr_dress_02, swing1LimitDegrees=5, swing2LimitDegrees=30, twistLimitDegrees=5}
 {bone1Name=dyn_r_fr_dress_04, bone2Name=dyn_r_fr_dress_03, swing1LimitDegrees=15, swing2LimitDegrees=30, twistLimitDegrees=15}
 {bone1Name=dyn_r_fr_dress_05, bone2Name=dyn_r_fr_dress_04, swing1LimitDegrees=15, swing2LimitDegrees=30, twistLimitDegrees=15}
 {
  bone1Name=spine_03
  bone2Name=spine_02
  swing1LimitDegrees=45
  swing1Motion=Free
  swing2LimitDegrees=45
  swing2Motion=Free
  twistLimitDegrees=45
  twistMotion=Free
 }
 {
  bone1Name=pelvis
  bone2Name=root
  swing1LimitDegrees=45
  swing1Motion=Free
  swing2LimitDegrees=45
  swing2Motion=Free
  twistLimitDegrees=45
  twistMotion=Free
 }
]
