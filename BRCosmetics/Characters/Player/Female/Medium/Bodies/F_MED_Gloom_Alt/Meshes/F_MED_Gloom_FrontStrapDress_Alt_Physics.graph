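# Physics asset for the front dress straps on the Gloom Alt female medium cosmetic body, defining bodies for secondary strap motion.
# Left and right three-link capsule chains hang from small MSTR bodies joined to the pelvis with free swing and twist.
# Chain links have narrow twist limits, with tighter swing at the left root.
# Pelvis, spine and thigh capsules form the body volume, with offset second capsules widening the thighs.
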PhysicsAsset /BRCosmetics/Characters/Player/Female/Medium/Bodies/F_MED_Gloom_Alt/Meshes/F_MED_Gloom_FrontStrapDress_Alt_Physics
bodies {
 dyn_l_fr_dressStrap_01 [{shapeType=Capsule, length=7.102, radius=2.4891, shapeName=dyn_l_fr_dressStrap_01_capsule, center=<4.8113 0 0>, rotation=<rot 90 0 0>}]
 dyn_l_fr_dressStrap_02 [{shapeType=Capsule, length=7.102, radius=2.4891, shapeName=dyn_l_fr_dressStrap_02_capsule, center=<4.8113 0 0>, rotation=<rot 90 0 0>}]
 dyn_l_fr_dressStrap_03 [{shapeType=Capsule, length=7.102, radius=2.4891, shapeName=dyn_l_fr_dressStrap_03_capsule, center=<4.8113 0 0>, rotation=<rot 90 0 0>}]
 dyn_l_fr_dressStrap_MSTR [{shapeType=Capsule, length=0.505, radius=0.505, shapeName=dyn_l_fr_dressStrap_MSTR_capsule, rotation=<rot 0 -90 0>}]
 dyn_r_fr_dressStrap_01 [{shapeType=Capsule, length=7.7749, radius=1.8163, shapeName=dyn_r_fr_dressStrap_01_capsule, center=<4.8113 0 2>, rotation=<rot 90 0 0>}]
 dyn_r_fr_dressStrap_02 [{shapeType=Capsule, length=7.761, radius=1.8301, shapeName=dyn_r_fr_dressStrap_02_capsule, center=<4.8113 0 0>, rotation=<rot 90 0 0>}]
 dyn_r_fr_dressStrap_03 [{shapeType=Capsule, length=7.761, radius=1.8301, shapeName=dyn_r_fr_dressStrap_03_capsule, center=<4.8113 0 0>, rotation=<rot 90 0 0>}]
 dyn_r_fr_dressStrap_MSTR [{shapeType=Capsule, length=0.505, radius=0.505, shapeName=dyn_r_fr_dressStrap_MSTR_capsule, rotation=<rot 0 -90 0>}]
 pelvis [{shapeType=Capsule, length=131.8578, radius=7.0303, center=<0 -1.2791 0>}]
 spine_01 [{shapeType=Capsule, length=135.5943, radius=6.0086}]
 spine_02 [{shapeType=Capsule, length=136.4493, radius=5.1536, center=<0 0.9981 0>}]
 spine_03 [{shapeType=Capsule, length=108.995, radius=32.6073, center=<3.6444 -27.5509 0>}]
 spine_04 [
  {shapeType=Capsule, length=6.8091, radius=6.7805, center=<0 5.5947 0>}
  {shapeType=Capsule, length=8.1789, radius=6.7805, shapeName=spine_04_capsule, center=<6.0515 3.3771 0>}
 ]
 spine_05 [{shapeType=Capsule, length=16.4921, radius=9.4395}]
 thigh_l [
  {shapeType=Capsule, length=39.1967, radius=10.652, center=<-26.2393 0 0>, rotation=<rot 90 0 0>}
  {shapeType=Capsule, length=36.9495, radius=12.8992, shapeName=thigh_l_capsule, center=<-28.6866 0 -17.6251>, rotation=<rot 90 0 0>}
 ]
 thigh_r [
  {shapeType=Capsule, length=40.5944, radius=9.2544, center=<25.6654 0 0>, rotation=<rot 90 0 0>}
  {shapeType=Capsule, length=37.6513, radius=12.1975, shapeName=thigh_r_capsule, center=<24.4354 0 19.3126>, rotation=<rot 90 0 0>}
 ]
}
constraints [
 {bone1Name=spine_02, bone2Name=spine_01, swing1LimitDegrees=45, swing2LimitDegrees=45, twistLimitDegrees=45}
 {bone1Name=spine_04, bone2Name=spine_03, swing1LimitDegrees=45, swing2LimitDegrees=45, twistLimitDegrees=45}
 {bone1Name=spine_05, bone2Name=spine_04, swing1LimitDegrees=45, swing2LimitDegrees=45, twistLimitDegrees=45}
 {bone1Name=thigh_l, bone2Name=pelvis, swing1LimitDegrees=45, swing2LimitDegrees=45, twistLimitDegrees=45}
 {bone1Name=thigh_r, bone2Name=pelvis, swing1LimitDegrees=45, swing2LimitDegrees=45, twistLimitDegrees=45}
 {
  bone1Name=spine_01
  bone2Name=pelvis
  swing1LimitDegrees=45
  swing1Motion=Free
  swing2LimitDegrees=45
  swing2Motion=Free
  twistLimitDegrees=45
  twistMotion=Free
 }
 {
  bone1Name=dyn_l_fr_dressStrap_MSTR
  bone2Name=pelvis
  swing1LimitDegrees=45
  swing1Motion=Free
  swing2LimitDegrees=45
  swing2Motion=Free
  twistLimitDegrees=45
  twistMotion=Free
 }
 {
  bone1Name=dyn_r_fr_dressStrap_MSTR
  bone2Name=pelvis
  swing1LimitDegrees=45
  swing1Motion=Free
  swing2LimitDegrees=45
  swing2Motion=Free
  twistLimitDegrees=45
  twistMotion=Free
 }
 {
  bone1Name=spine_03
  bone2Name=spine_02
  swing1LimitDegrees=45
  swing1Motion=Free
  swing2LimitDegrees=45
  swing2Motion=Free
  twistLimitDegrees=45
  twistMotion=Free
 }
 {bone1Name=dyn_l_fr_dressStrap_01, bone2Name=dyn_l_fr_dressStrap_MSTR, swing1LimitDegrees=20, swing2LimitDegrees=30, twistLimitDegrees=10}
 {bone1Name=dyn_l_fr_dressStrap_02, bone2Name=dyn_l_fr_dressStrap_01, swing1LimitDegrees=45, swing2LimitDegrees=45, twistLimitDegrees=10}
 {bone1Name=dyn_l_fr_dressStrap_03, bone2Name=dyn_l_fr_dressStrap_02, swing1LimitDegrees=45, swing2LimitDegrees=45, twistLimitDegrees=10}
 {bone1Name=dyn_r_fr_dressStrap_01, bone2Name=dyn_r_fr_dressStrap_MSTR, swing1LimitDegrees=45, swing2LimitDegrees=45, twistLimitDegrees=10}
 {bone1Name=dyn_r_fr_dressStrap_02, bone2Name=dyn_r_fr_dressStrap_01, swing1LimitDegrees=45, swing2LimitDegrees=45, twistLimitDegrees=10}
 {bone1Name=dyn_r_fr_dressStrap_03, bone2Name=dyn_r_fr_dressStrap_02, swing1LimitDegrees=45, swing2LimitDegrees=45, twistLimitDegrees=10}
]
