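# Animation Blueprint for the female medium Gloom Alt body cosmetic, driving the secondary motion of its dress panels and straps.
# It copies the parent mesh pose through a Control Rig, then corrects dress bones with weighted constraints, Look Ats and additive offsets, some driven by crouch, jump and thigh rotation.
# RigidBody simulations branch from a cached PreSim pose, using separate front, back and strap physics assets.
# Bool blends select between simulation setups for front end, emotes, acceleration and skydiving.
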AnimBlueprint /BRCosmetics/Characters/Player/Female/Medium/Bodies/F_MED_Gloom_Alt/Meshes/F_MED_Gloom_Alt_AnimBP
target_skeleton=&/F_MED_Gloom_Alt_Skeleton
graphs [
 AnimGraph {nodes [
   n1 = "Constraint - dyn_l_fr_dressStrap_MSTR (PPPPPPPPPP)"!(
    ComponentPose
    Alpha=1
    ConstraintWeights_0=0
    ConstraintWeights_1:float=0
    ConstraintWeights_2:float=0.0782
    ConstraintWeights_3:float=0.0692
    ConstraintWeights_4:float=0.1695
    ConstraintWeights_5:float=0.002
    ConstraintWeights_6:float=0.0439
    ConstraintWeights_7:float=0.0098
    ConstraintWeights_8:float=0.5721
    ConstraintWeights_9:float=0.0552
   ) -> (Pose) {
    alphaBoolBlend {blendOption=Linear}
    boneToModify {boneName=dyn_l_fr_dressStrap_MSTR}
    constraintSetup [
     {targetBone {boneName=spine_02}}
     {targetBone {boneName=spine_03}}
     {targetBone {boneName=spine_04}}
     {targetBone {boneName=spine_05}}
     {targetBone {boneName=clavicle_l}}
     {targetBone {boneName=deform_upperarm_bk_base_l}}
     {targetBone {boneName=deform_upperarm_fr_l}}
     {targetBone {boneName=deform_upperarm_up_base_l}}
     {targetBone {boneName=upperarm_twist_02_l}}
     {targetBone {boneName=deform_pec_l}}
    ]
    constraintWeights [1,1,1,1,1,1,1,1,1,1]
   }
   "Constraint - dyn_r_fr_dressStrap_MSTR (PPPPPPP)"!(
    ComponentPose=@n1
    Alpha=1
    ConstraintWeights_0=0.0502
    ConstraintWeights_1:float=0.0742
    ConstraintWeights_2:float=0.2255
    ConstraintWeights_3:float=0.2106
    ConstraintWeights_4:float=0.0448
    ConstraintWeights_5:float=0.3797
    ConstraintWeights_6:float=0.015
   ) -> (Pose) {
    alphaBoolBlend {blendOption=Linear}
    boneToModify {boneName=dyn_r_fr_dressStrap_MSTR}
    constraintSetup [
     {targetBone {boneName=spine_04}}
     {targetBone {boneName=spine_05}}
     {targetBone {boneName=clavicle_r}}
     {targetBone {boneName=deform_upperarm_fr_base_r}}
     {targetBone {boneName=deform_upperarm_up_base_r}}
     {targetBone {boneName=upperarm_twist_02_r}}
     {targetBone {boneName=deform_pec_base_r}}
    ]
    constraintWeights [1,1,1,1,1,1,1]
   }
   n2 = Get CrouchWalking() -> (CrouchWalking:bool)
   n3 = "Constraint - dyn_r_fr_dress_03 (RR)"!(ComponentPose, bAlphaBoolEnabled:bool=@n2, ConstraintWeights_0=0, ConstraintWeights_1:float=1) -> (Pose) {
    alphaInputType=Bool
    alphaBoolBlend {blendInTime=0.1, blendOption=Linear, blendOutTime=0.1}
    boneToModify {boneName=dyn_r_fr_dress_03}
    constraintSetup [{transformType=Rotation, targetBone {boneName=thigh_l}}, {transformType=Rotation, targetBone {boneName=thigh_r}}]
    constraintWeights [1,1]
   }
   "Constraint - dyn_l_fr_dress_03 (RR)"!(ComponentPose=@n3, bAlphaBoolEnabled:bool=@n2, ConstraintWeights_0=1, ConstraintWeights_1:float=0) -> (Pose) {
    alphaInputType=Bool
    alphaBoolBlend {blendInTime=0.1, blendOption=Linear, blendOutTime=0.1}
    boneToModify {boneName=dyn_l_fr_dress_03}
    constraintSetup [{transformType=Rotation, targetBone {boneName=thigh_l}}, {transformType=Rotation, targetBone {boneName=thigh_r}}]
    constraintWeights [1,1]
   }
   "Transform (Modify) Bone - Bone: dyn_r_bk_dressStrap_MSTR"!(ComponentPose, bAlphaBoolEnabled:bool=True, Translation="0,0,0", Rotation="0,-60,0", Scale="0.8,1,1") -> (Pose) {
    alphaInputType=Bool
    rotationMode=BMM_Additive
    rotationSpace=BCS_BoneSpace
    scaleMode=BMM_Additive
    scaleSpace=BCS_BoneSpace
    translationMode=BMM_Additive
    translationSpace=BCS_BoneSpace
    alphaBoolBlend {blendInTime=0.1, blendOption=Linear, blendOutTime=0.1}
    boneToModify {boneName=dyn_r_bk_dressStrap_MSTR}
    rotation {yaw=-60}
    scale {x=0.8}
   }
   n4 = RigidBody!(
    ComponentPose=Local To Component(LocalPose=Use cached pose 'PreSim'())
    Alpha
    OverrideWorldGravity:Vector="0,0,0"
    ExternalForce
    ComponentLinearAccScale:Vector="0,0,0"
   ) -> (Pose) {
    bClampLinearTranslationLimitToRefPose=true
    bOverrideWorldGravity=true
    overridePhysicsAsset=&/F_MED_Gloom_FrontDress_Alt_Emote_Physics
    simulationSpace=BaseBoneSpace
    alphaBoolBlend {blendOption=Linear}
    baseBoneRef {boneName=root}
    componentAppliedLinearAccClamp=<5000 5000 5000>
    componentLinearAccScale {z=0.45}
    componentLinearVelScale=<0.4 0.4 0.55>
    overrideWorldGravity {z=-980}
    simSpaceSettings {
     maxAngularAcceleration=1000
     maxAngularVelocity=3
     maxLinearAcceleration=700
     maxLinearVelocity=500
     velocityScaleZ=0
     externalLinearDragV=<1 1>
    }
   }
   n5 = "Constraint - dyn_l_fr_dress_05 (TT)"!(ComponentPose=@n4, Alpha=0.5, ConstraintWeights_0=0.8, ConstraintWeights_1:float=0.2) -> (Pose) {
    alphaBoolBlend {blendOption=Linear}
    boneToModify {boneName=dyn_l_fr_dress_05}
    constraintSetup [{transformType=Translation, targetBone {boneName=calf_l}}, {transformType=Translation, targetBone {boneName=thigh_r}}]
    constraintWeights [1,1]
   }
   n6 = "Constraint - dyn_r_fr_dress_05 (TT)"!(ComponentPose=@n5, Alpha=0.5, ConstraintWeights_0=0.8, ConstraintWeights_1:float=0.2) -> (Pose) {
    alphaBoolBlend {blendOption=Linear}
    boneToModify {boneName=dyn_r_fr_dress_05}
    constraintSetup [{transformType=Translation, targetBone {boneName=calf_r}}, {transformType=Translation, targetBone {boneName=thigh_r}}]
    constraintWeights [1,1]
   }
   n7 = "Constraint - dyn_l_fr_dressStrap_MSTR (PPPPPPPPPP)"(
    ComponentPose=Local To Component(LocalPose=Use cached pose 'PreSim'())
    Alpha=1
    ConstraintWeights_0=0
    ConstraintWeights_1=0
    ConstraintWeights_2=0.1236
    ConstraintWeights_3=0.1085
    ConstraintWeights_4=0.1925
    ConstraintWeights_5=0.0016
    ConstraintWeights_6=0.0528
    ConstraintWeights_7=0.0057
    ConstraintWeights_8=0.4235
    ConstraintWeights_9=0.0917
   ) {
    alphaBoolBlend {blendOption=Linear}
    boneToModify {boneName=dyn_l_fr_dressStrap_MSTR}
    constraintSetup [
     {targetBone {boneName=spine_02}}
     {targetBone {boneName=spine_03}}
     {targetBone {boneName=spine_04}}
     {targetBone {boneName=spine_05}}
     {targetBone {boneName=clavicle_l}}
     {targetBone {boneName=deform_upperarm_bk_base_l}}
     {targetBone {boneName=deform_upperarm_fr_l}}
     {targetBone {boneName=deform_upperarm_up_base_l}}
     {targetBone {boneName=upperarm_twist_02_l}}
     {targetBone {boneName=deform_pec_l}}
    ]
    constraintWeights [1,1,1,1,1,1,1,1,1,1]
   }
   n8 = "Constraint - dyn_r_fr_dressStrap_MSTR (PPPPPPPP)"!(
    ComponentPose=@n7
    Alpha=1
    ConstraintWeights_0=0.1143
    ConstraintWeights_1:float=0.1266
    ConstraintWeights_2:float=0.2486
    ConstraintWeights_3:float=0.0003
    ConstraintWeights_4:float=0.1638
    ConstraintWeights_5:float=0.0069
    ConstraintWeights_6:float=0.3012
    ConstraintWeights_7:float=0.0382
   ) -> (Pose) {
    alphaBoolBlend {blendOption=Linear}
    boneToModify {boneName=dyn_r_fr_dressStrap_MSTR}
    constraintSetup [
     {targetBone {boneName=spine_04}}
     {targetBone {boneName=spine_05}}
     {targetBone {boneName=clavicle_r}}
     {targetBone {boneName=deform_upperarm_bk_r}}
     {targetBone {boneName=deform_upperarm_fr_base_r}}
     {targetBone {boneName=deform_upperarm_up_base_r}}
     {targetBone {boneName=upperarm_twist_02_r}}
     {targetBone {boneName=deform_pec_base_r}}
    ]
    constraintWeights [1,1,1,1,1,1,1,1]
   }
   n9 = "Transform (Modify) Bone - Bone: dyn_sideSkirt"(ComponentPose=@n8, Alpha=1, Translation="0,-1,0", Rotation="-0.006,-3.793,0.000", Scale="1,1,1") {
    rotationMode=BMM_Additive
    rotationSpace=BCS_BoneSpace
    scaleMode=BMM_Additive
    scaleSpace=BCS_BoneSpace
    translationMode=BMM_Additive
    translationSpace=BCS_BoneSpace
    alphaBoolBlend {blendOption=Linear}
    boneToModify {boneName=dyn_sideSkirt}
    rotation=<rot -0.006 -3.793 0>
    translation {y=-1}
   }
   n10 = "Transform (Modify) Bone - Bone: dyn_l_fr_dressStrap_MSTR"(ComponentPose=@n9, Alpha=1, Translation="-0.791,-0.001,0.830", Rotation="-21.945,0.016,-3.698", Scale="1,1,1") {
    rotationMode=BMM_Additive
    rotationSpace=BCS_BoneSpace
    scaleMode=BMM_Additive
    scaleSpace=BCS_BoneSpace
    translationMode=BMM_Additive
    translationSpace=BCS_BoneSpace
    alphaBoolBlend {blendOption=Linear}
    boneToModify {boneName=dyn_l_fr_dressStrap_MSTR}
    rotation=<rot -21.945 0.016 -3.698>
    translation=<-0.791 -0.001 0.83>
   }
   n11 = "Transform (Modify) Bone - Bone: dyn_l_fr_dressStrap_02"(ComponentPose=@n10, Alpha=1, Translation="1,0,0", Rotation="0,0,0", Scale="0.8,1,1") {
    rotationMode=BMM_Additive
    rotationSpace=BCS_BoneSpace
    scaleMode=BMM_Additive
    scaleSpace=BCS_BoneSpace
    translationMode=BMM_Additive
    translationSpace=BCS_BoneSpace
    alphaBoolBlend {blendOption=Linear}
    boneToModify {boneName=dyn_l_fr_dressStrap_02}
    scale {x=0.8}
    translation {x=1}
   }
   n12 = "Transform (Modify) Bone - Bone: dyn_l_fr_dressStrap_03"(ComponentPose=@n11, Alpha=1, Translation="-26.839,185.664,39.970", Rotation="9.348,26.605,14.893", Scale="0.8,1,1") {
    rotationMode=BMM_Additive
    rotationSpace=BCS_BoneSpace
    scaleMode=BMM_Additive
    scaleSpace=BCS_BoneSpace
    translationMode=BMM_Additive
    translationSpace=BCS_BoneSpace
    alphaBoolBlend {blendOption=Linear}
    boneToModify {boneName=dyn_l_fr_dressStrap_03}
    rotation=<rot 9.348 26.605 14.893>
    scale {x=0.8}
    translation=<-26.839 185.664 39.97>
   }
   n13 = Get bIsPlayingEmote()
   n14 = Look At - Bone: dyn_r_fr_dress_03!(ComponentPose=@n12, bAlphaBoolEnabled:bool=@n13) -> (Pose) {
    alphaInputType=Bool
    alphaBoolBlend {blendInTime=0.1, blendOption=Linear, blendOutTime=0.1}
    alphaScaleBiasClamp {bInterpResult=true}
    boneToModify {boneName=dyn_r_fr_dress_03}
    lookAtLocation=<-1.7847 15 2.562>
    lookAtTarget {boneReference {boneName=calf_r}}
    lookAt_Axis {axis=<1 0>}
    lookUp_Axis {axis {x=0}}
   }
   n15 = Look At - Bone: dyn_l_fr_dress_03!(ComponentPose=@n14, bAlphaBoolEnabled:bool=@n13) -> (Pose) {
    alphaInputType=Bool
    alphaBoolBlend {blendInTime=0.1, blendOption=Linear, blendOutTime=0.1}
    alphaScaleBiasClamp {bInterpResult=true}
    boneToModify {boneName=dyn_l_fr_dress_03}
    lookAtLocation=<-1.7847 15 2.562>
    lookAtTarget {boneReference {boneName=calf_l}}
    lookAt_Axis {axis=<1 0>}
    lookUp_Axis {axis {x=0}}
   }
   n16 = RigidBody(ComponentPose=@n15, OverrideWorldGravity="0,0,0", ComponentLinearAccScale="0,0,0") {
    bClampLinearTranslationLimitToRefPose=true
    bOverrideWorldGravity=true
    overridePhysicsAsset=&/F_MED_Gloom_FrontStrapDress_Alt_Physics
    simulationSpace=BaseBoneSpace
    alphaBoolBlend {blendOption=Linear}
    baseBoneRef {boneName=spine_05}
    componentAppliedLinearAccClamp=<1000 1000 1000>
    componentLinearAccScale {z=0.45}
    componentLinearVelScale=<0.3 0.3 0.3>
    overrideWorldGravity {z=-980}
    simSpaceSettings {
     maxAngularAcceleration=1000
     maxAngularVelocity=3
     maxLinearAcceleration=700
     maxLinearVelocity=500
     velocityScaleZ=0
     worldAlpha=0.1
     externalLinearDragV=<1 1>
    }
   }
   n17 = RigidBody(ComponentPose=@n16, OverrideWorldGravity="0,0,0", ComponentLinearAccScale="0.2,0.2,0.3") {
    bClampLinearTranslationLimitToRefPose=true
    bOverrideWorldGravity=true
    overridePhysicsAsset=&/F_MED_Gloom_BackStrapDress_Alt_Physics
    simulationSpace=BaseBoneSpace
    alphaBoolBlend {blendOption=Linear}
    baseBoneRef {boneName=spine_05}
    componentAppliedLinearAccClamp=<1000 1000 3500>
    componentLinearAccScale {z=0.45}
    componentLinearVelScale=<0.2 0.2 0.3>
    overrideWorldGravity {z=-980}
    simSpaceSettings {
     maxAngularAcceleration=1000
     maxAngularVelocity=3
     maxLinearAcceleration=700
     maxLinearVelocity=500
     velocityScaleZ=0
     worldAlpha=0.1
     externalLinearDragV=<1 1>
    }
   }
   n18 = RigidBody(ComponentPose=@n17, OverrideWorldGravity="0,0,0", ComponentLinearAccScale=Get FrontDress Velocity() -> (FrontDress Velocity:Vector)) {
    bClampLinearTranslationLimitToRefPose=true
    bOverrideWorldGravity=true
    overridePhysicsAsset=&/F_MED_Gloom_FrontDress_Alt_FE_Physics
    simulationSpace=BaseBoneSpace
    alphaBoolBlend {blendOption=Linear}
    baseBoneRef {boneName=root}
    componentAppliedLinearAccClamp=<5000 5000 5000>
    componentLinearAccScale {z=0.45}
    componentLinearVelScale=<0.4 0.4 0.55>
    overrideWorldGravity {z=-980}
    simSpaceSettings {
     maxAngularAcceleration=1000
     maxAngularVelocity=3
     maxLinearAcceleration=700
     maxLinearVelocity=500
     velocityScaleZ=0
     externalLinearDragV=<1 1>
    }
   }
   n19 = RigidBody(ComponentPose=@n18, OverrideWorldGravity="0,0,0", ComponentLinearAccScale="0.8,0.3,0.8") {
    bClampLinearTranslationLimitToRefPose=true
    bOverrideWorldGravity=true
    overridePhysicsAsset=&/F_MED_Gloom_BackDress_Alt_Physics
    simulationSpace=BaseBoneSpace
    alphaBoolBlend {blendOption=Linear}
    baseBoneRef {boneName=root}
    componentAppliedLinearAccClamp=<2000 2000 6000>
    componentLinearAccScale {z=0.45}
    componentLinearVelScale=<0.8 0.3 1>
    overrideWorldGravity {z=-980}
    simSpaceSettings {
     maxAngularAcceleration=1000
     maxAngularVelocity=3
     maxLinearAcceleration=700
     maxLinearVelocity=500
     velocityScaleZ=0
     worldAlpha=0.2
     externalLinearDragV=<1 1>
    }
   }
   n20 = Get NewNotEmoting() -> (NewNotEmoting:bool)
   n21 = "Transform (Modify) Bone - Bone: dyn_l_fr_dress_02"!(ComponentPose=@n19, bAlphaBoolEnabled:bool=@n20, Translation="0,0,9", Rotation="0,0,0", Scale="1,1,1") -> (Pose) {
    alphaInputType=Bool
    rotationMode=BMM_Additive
    rotationSpace=BCS_BoneSpace
    scaleMode=BMM_Additive
    scaleSpace=BCS_BoneSpace
    translationMode=BMM_Additive
    translationSpace=BCS_BoneSpace
    alphaBoolBlend {blendOption=Linear}
    boneToModify {boneName=dyn_l_fr_dress_02}
    translation {z=9}
   }
   n22 = "Transform (Modify) Bone - Bone: dyn_l_fr_dress_03"!(ComponentPose=@n21, bAlphaBoolEnabled:bool=@n20, Translation="0,0,-9", Rotation="0,0,0", Scale="1,1,1") -> (Pose) {
    alphaInputType=Bool
    rotationMode=BMM_Additive
    rotationSpace=BCS_BoneSpace
    scaleMode=BMM_Additive
    scaleSpace=BCS_BoneSpace
    translationMode=BMM_Additive
    translationSpace=BCS_BoneSpace
    alphaBoolBlend {blendOption=Linear}
    boneToModify {boneName=dyn_l_fr_dress_03}
    translation {z=-9}
   }
   n23 = "Transform (Modify) Bone - Bone: dyn_l_fr_dress_04"!(ComponentPose=@n22, bAlphaBoolEnabled:bool=@n20, Translation="0,0,-5", Rotation="0,0,0", Scale="1,1,1") -> (Pose) {
    alphaInputType=Bool
    rotationMode=BMM_Additive
    rotationSpace=BCS_BoneSpace
    scaleMode=BMM_Additive
    scaleSpace=BCS_BoneSpace
    translationMode=BMM_Additive
    translationSpace=BCS_BoneSpace
    alphaBoolBlend {blendOption=Linear}
    boneToModify {boneName=dyn_l_fr_dress_04}
    translation {z=-5}
   }
   n24 = "Constraint - dyn_l_fr_dressStrap_02 (R)"(ComponentPose=@n23, Alpha=0.8, ConstraintWeights_0=1) {
    alphaBoolBlend {blendOption=Linear}
    boneToModify {boneName=dyn_l_fr_dressStrap_02}
    constraintSetup [{transformType=Rotation, targetBone {boneName=dyn_l_fr_dress_03}}]
    constraintWeights [1]
   }
   n25 = "Constraint - dyn_l_fr_dressStrap_03 (R)"(ComponentPose=@n24, Alpha=1, ConstraintWeights_0=1) {
    alphaBoolBlend {blendOption=Linear}
    boneToModify {boneName=dyn_l_fr_dressStrap_03}
    constraintSetup [{transformType=Rotation, targetBone {boneName=dyn_l_fr_dress_03}}]
    constraintWeights [1]
   }
   n26 = RigidBody(ComponentPose=Local To Component(LocalPose=Use cached pose 'PreSim'()), OverrideWorldGravity="0,0,0", ComponentLinearAccScale="0.4,0.3,0.4") {
    bClampLinearTranslationLimitToRefPose=true
    bOverrideWorldGravity=true
    overridePhysicsAsset=&F_MED_Gloom_BackDress_Alt_Physics
    simulationSpace=BaseBoneSpace
    alphaBoolBlend {blendOption=Linear}
    baseBoneRef {boneName=root}
    componentAppliedLinearAccClamp=<2000 2000 2000>
    componentLinearAccScale {z=0.45}
    componentLinearVelScale=<0.4 0.3 0.4>
    overrideWorldGravity {z=-980}
    simSpaceSettings {
     maxAngularAcceleration=1000
     maxAngularVelocity=3
     maxLinearAcceleration=700
     maxLinearVelocity=500
     velocityScaleZ=0
     externalLinearDragV=<1 1>
    }
   }
   n27 = "Transform (Modify) Bone - Bone: dyn_l_bk_dressStrap_MSTR"!(ComponentPose=@n26, bAlphaBoolEnabled:bool=True, Translation="0,0,1", Rotation="40,0,0", Scale="1,1,1") -> (Pose) {
    alphaInputType=Bool
    rotationMode=BMM_Additive
    rotationSpace=BCS_BoneSpace
    scaleMode=BMM_Additive
    scaleSpace=BCS_BoneSpace
    translationMode=BMM_Additive
    translationSpace=BCS_BoneSpace
    alphaBoolBlend {blendInTime=0.1, blendOption=Linear, blendOutTime=0.1}
    boneToModify {boneName=dyn_l_bk_dressStrap_MSTR}
    rotation {pitch=40}
    translation {z=1}
   }
   n28 = "Transform (Modify) Bone - Bone: dyn_l_bk_dressStrap_MSTR"(ComponentPose=@n27, bAlphaBoolEnabled=True, Translation="0,0,0", Rotation="0,0,0", Scale="1,1,1") {
    alphaInputType=Bool
    rotationMode=BMM_Additive
    rotationSpace=BCS_BoneSpace
    scaleMode=BMM_Additive
    scaleSpace=BCS_BoneSpace
    translationMode=BMM_Additive
    translationSpace=BCS_BoneSpace
    alphaBoolBlend {blendInTime=0.1, blendOption=Linear, blendOutTime=0.1}
    boneToModify {boneName=dyn_l_bk_dressStrap_MSTR}
   }
   n29 = "Transform (Modify) Bone - Bone: dyn_l_bk_dress_03"!(ComponentPose=@n28, bAlphaBoolEnabled:bool=True, Translation="7,0,0", Rotation="0,0,0", Scale="1,1,1") -> (Pose) {
    alphaInputType=Bool
    rotationMode=BMM_Additive
    rotationSpace=BCS_BoneSpace
    scaleMode=BMM_Additive
    scaleSpace=BCS_BoneSpace
    translationMode=BMM_Additive
    translationSpace=BCS_BoneSpace
    alphaBoolBlend {blendInTime=0.1, blendOption=Linear, blendOutTime=0.1}
    boneToModify {boneName=dyn_l_bk_dress_03}
    translation {x=7}
   }
   n30 = "Transform (Modify) Bone - Bone: dyn_r_bk_dress_03"!(ComponentPose=@n29, bAlphaBoolEnabled:bool=True, Translation="7,0,0", Rotation="0,0,0", Scale="1,1,1") -> (Pose) {
    alphaInputType=Bool
    rotationMode=BMM_Additive
    rotationSpace=BCS_BoneSpace
    scaleMode=BMM_Additive
    scaleSpace=BCS_BoneSpace
    translationMode=BMM_Additive
    translationSpace=BCS_BoneSpace
    alphaBoolBlend {blendInTime=0.1, blendOption=Linear, blendOutTime=0.1}
    boneToModify {boneName=dyn_r_bk_dress_03}
    translation {x=7}
   }
   n31 = Component To Local(ComponentPose=@n30)
   n32 = RigidBody(ComponentPose=Local To Component(LocalPose=Use cached pose 'PreSim'()), OverrideWorldGravity="0,0,0", ComponentLinearAccScale="0.8,0.3,0.8") {
    bClampLinearTranslationLimitToRefPose=true
    bOverrideWorldGravity=true
    overridePhysicsAsset=&F_MED_Gloom_BackDress_Alt_Physics
    simulationSpace=BaseBoneSpace
    alphaBoolBlend {blendOption=Linear}
    baseBoneRef {boneName=pelvis}
    componentAppliedLinearAccClamp=<2000 2000 6000>
    componentLinearAccScale {z=0.45}
    componentLinearVelScale=<0.8 0.3 1>
    overrideWorldGravity {z=-980}
    simSpaceSettings {
     maxAngularAcceleration=1000
     maxAngularVelocity=3
     maxLinearAcceleration=700
     maxLinearVelocity=500
     velocityScaleZ=0
     worldAlpha=0.2
     externalLinearDragV=<1 1>
    }
   }
   n33 = Component To Local(ComponentPose=@n32)
   n34 = OR Boolean!(A=Get bIsAccelerating2D(), B=Get bIsPlayingEmote(), C:bool=Get bIsSkyDiving()) -> (ReturnValue)
   n35 = RigidBody(
    ComponentPose=Local To Component(LocalPose=Blend Poses by bool(BlendPose_0=@n31, BlendPose_1=@n33, BlendTime_0=1, BlendTime_1=1, bActiveValue=@n34) {blendType=Linear})
    OverrideWorldGravity="0,0,0"
    ComponentLinearAccScale="0,0,0"
   ) {
    bClampLinearTranslationLimitToRefPose=true
    bOverrideWorldGravity=true
    overridePhysicsAsset=&F_MED_Gloom_FrontStrapDress_Alt_Physics
    simulationSpace=BaseBoneSpace
    alphaBoolBlend {blendOption=Linear}
    baseBoneRef {boneName=spine_05}
    componentAppliedLinearAccClamp=<1000 1000 1000>
    componentLinearAccScale {z=0.45}
    componentLinearVelScale=<0.3 0.3 0.3>
    overrideWorldGravity {z=-980}
    simSpaceSettings {
     maxAngularAcceleration=1000
     maxAngularVelocity=3
     maxLinearAcceleration=700
     maxLinearVelocity=500
     velocityScaleZ=0
     worldAlpha=0.1
     externalLinearDragV=<1 1>
    }
   }
   n36 = RigidBody(ComponentPose=@n35, OverrideWorldGravity="0,0,0", ComponentLinearAccScale="0.2,0.2,0.3") {
    bClampLinearTranslationLimitToRefPose=true
    bOverrideWorldGravity=true
    overridePhysicsAsset=&F_MED_Gloom_BackStrapDress_Alt_Physics
    simulationSpace=BaseBoneSpace
    alphaBoolBlend {blendOption=Linear}
    baseBoneRef {boneName=spine_05}
    componentAppliedLinearAccClamp=<1000 1000 3500>
    componentLinearAccScale {z=0.45}
    componentLinearVelScale=<0.2 0.2 0.3>
    overrideWorldGravity {z=-980}
    simSpaceSettings {
     maxAngularAcceleration=1000
     maxAngularVelocity=3
     maxLinearAcceleration=700
     maxLinearVelocity=500
     velocityScaleZ=0
     worldAlpha=0.1
     externalLinearDragV=<1 1>
    }
   }
   n37 = RigidBody(ComponentPose=@n36, OverrideWorldGravity="0,0,0", ComponentLinearAccScale=Get FrontDress Velocity()) {
    bClampLinearTranslationLimitToRefPose=true
    bOverrideWorldGravity=true
    overridePhysicsAsset=&/F_MED_Gloom_FrontDress_Alt_Physics
    simulationSpace=BaseBoneSpace
    alphaBoolBlend {blendOption=Linear}
    baseBoneRef {boneName=root}
    componentAppliedLinearAccClamp=<5000 5000 5000>
    componentLinearAccScale {z=0.45}
    componentLinearVelScale=<0.4 0.4 0.55>
    overrideWorldGravity {z=-980}
    simSpaceSettings {
     maxAngularAcceleration=1000
     maxAngularVelocity=3
     maxLinearAcceleration=700
     maxLinearVelocity=500
     velocityScaleZ=0
     externalLinearDragV=<1 1>
    }
   }
   n38 = RigidBody(ComponentPose=Local To Component(LocalPose=Use cached pose 'PreSim'()), OverrideWorldGravity="0,0,0", ComponentLinearAccScale="0,0,0") {
    bClampLinearTranslationLimitToRefPose=true
    bOverrideWorldGravity=true
    overridePhysicsAsset=&F_MED_Gloom_FrontDress_Alt_Emote_Physics
    simulationSpace=BaseBoneSpace
    alphaBoolBlend {blendOption=Linear}
    baseBoneRef {boneName=root}
    componentAppliedLinearAccClamp=<5000 5000 5000>
    componentLinearAccScale {z=0.45}
    componentLinearVelScale=<0.4 0.4 0.55>
    overrideWorldGravity {z=-980}
    simSpaceSettings {
     maxAngularAcceleration=1000
     maxAngularVelocity=3
     maxLinearAcceleration=700
     maxLinearVelocity=500
     velocityScaleZ=0
     externalLinearDragV=<1 1>
    }
   }
   n39 = "Constraint - dyn_l_fr_dress_05 (TT)"(ComponentPose=@n38, Alpha=0.5, ConstraintWeights_0=0.8, ConstraintWeights_1=0.2) {
    alphaBoolBlend {blendOption=Linear}
    boneToModify {boneName=dyn_l_fr_dress_05}
    constraintSetup [{transformType=Translation, targetBone {boneName=calf_l}}, {transformType=Translation, targetBone {boneName=thigh_r}}]
    constraintWeights [1,1]
   }
   n40 = "Constraint - dyn_r_fr_dress_05 (TT)"(ComponentPose=@n39, Alpha=0.5, ConstraintWeights_0=0.8, ConstraintWeights_1=0.2) {
    alphaBoolBlend {blendOption=Linear}
    boneToModify {boneName=dyn_r_fr_dress_05}
    constraintSetup [{transformType=Translation, targetBone {boneName=calf_r}}, {transformType=Translation, targetBone {boneName=thigh_r}}]
    constraintWeights [1,1]
   }
   n41 = Blend Poses by bool(
    BlendPose_0=Blend Poses by bool(BlendPose_0=Component To Local(ComponentPose=@n6), BlendPose_1=Component To Local(ComponentPose=@n25), bActiveValue=Get bIsPlayingEmote()) {blendType=Linear}
    BlendPose_1=Blend Poses by bool(BlendPose_0=Component To Local(ComponentPose=@n40), BlendPose_1=Component To Local(ComponentPose=@n37), bActiveValue=Get bIsPlayingEmote()) {blendType=Linear}
    bActiveValue=Get bIsInFrontEnd()
   ) {blendType=Linear}
   n42 = Control Rig(Source=Copy Pose From Mesh() {bCopyCurves=true, bUseAttachedParent=true}) {
    alphaBoolBlend {blendOption=Linear}
    controlRigAssetReference {blueprintRigClass=&/F_MED_Gloom_Alt_AltSkin_CtrlRig.F_MED_Gloom_Alt_AltSkin_CtrlRig_C}
   }
   n43 = "Constraint - dyn_l_fr_dressStrap_MSTR (PPPPPPPPPP)"(
    ComponentPose=Local To Component(LocalPose=@n42)
    Alpha=1
    ConstraintWeights_0=0
    ConstraintWeights_1=0
    ConstraintWeights_2=0.1236
    ConstraintWeights_3=0.1085
    ConstraintWeights_4=0.1925
    ConstraintWeights_5=0.0016
    ConstraintWeights_6=0.0528
    ConstraintWeights_7=0.0057
    ConstraintWeights_8=0.4235
    ConstraintWeights_9=0.0917
   ) {
    alphaBoolBlend {blendOption=Linear}
    boneToModify {boneName=dyn_l_fr_dressStrap_MSTR}
    constraintSetup [
     {targetBone {boneName=spine_02}}
     {targetBone {boneName=spine_03}}
     {targetBone {boneName=spine_04}}
     {targetBone {boneName=spine_05}}
     {targetBone {boneName=clavicle_l}}
     {targetBone {boneName=deform_upperarm_bk_base_l}}
     {targetBone {boneName=deform_upperarm_fr_l}}
     {targetBone {boneName=deform_upperarm_up_base_l}}
     {targetBone {boneName=upperarm_twist_02_l}}
     {targetBone {boneName=deform_pec_l}}
    ]
    constraintWeights [1,1,1,1,1,1,1,1,1,1]
   }
   n44 = "Constraint - dyn_r_fr_dressStrap_MSTR (PPPPPPPP)"(
    ComponentPose=@n43
    Alpha=1
    ConstraintWeights_0=0.1143
    ConstraintWeights_1=0.1266
    ConstraintWeights_2=0.2486
    ConstraintWeights_3=0.0003
    ConstraintWeights_4=0.1638
    ConstraintWeights_5=0.0069
    ConstraintWeights_6=0.3012
    ConstraintWeights_7=0.0382
   ) {
    alphaBoolBlend {blendOption=Linear}
    boneToModify {boneName=dyn_r_fr_dressStrap_MSTR}
    constraintSetup [
     {targetBone {boneName=spine_04}}
     {targetBone {boneName=spine_05}}
     {targetBone {boneName=clavicle_r}}
     {targetBone {boneName=deform_upperarm_bk_r}}
     {targetBone {boneName=deform_upperarm_fr_base_r}}
     {targetBone {boneName=deform_upperarm_up_base_r}}
     {targetBone {boneName=upperarm_twist_02_r}}
     {targetBone {boneName=deform_pec_base_r}}
    ]
    constraintWeights [1,1,1,1,1,1,1,1]
   }
   n45 = "Transform (Modify) Bone - Bone: dyn_l_fr_dressStrap_01"!(ComponentPose=@n44, bAlphaBoolEnabled:bool=True, Translation="0,0,0", Rotation="27.379,21.809,11.853", Scale="1,1,1") -> (Pose) {
    alphaInputType=Bool
    rotationMode=BMM_Additive
    rotationSpace=BCS_BoneSpace
    scaleMode=BMM_Additive
    scaleSpace=BCS_BoneSpace
    translationMode=BMM_Additive
    translationSpace=BCS_BoneSpace
    alphaBoolBlend {blendOption=Linear}
    boneToModify {boneName=dyn_l_fr_dressStrap_01}
    rotation=<rot 27.379 21.809 11.853>
   }
   n46 = "Transform (Modify) Bone - Bone: dyn_r_fr_dressStrap_01"!(ComponentPose=@n45, bAlphaBoolEnabled:bool=True, Translation="0,0,0", Rotation="9.270,9.298,1.521", Scale="1,1,1") -> (Pose) {
    alphaInputType=Bool
    rotationMode=BMM_Additive
    rotationSpace=BCS_BoneSpace
    scaleMode=BMM_Additive
    scaleSpace=BCS_BoneSpace
    translationMode=BMM_Additive
    translationSpace=BCS_BoneSpace
    alphaBoolBlend {blendOption=Linear}
    boneToModify {boneName=dyn_r_fr_dressStrap_01}
    rotation=<rot 9.27 9.298 1.521>
   }
   n47 = Get NewNotFrontEnd() -> (NewNotFrontEnd:bool)
   n48 = Look At - Bone: dyn_l_fr_dressStrap_MSTR!(ComponentPose=@n46, bAlphaBoolEnabled:bool=@n47) -> (Pose) {
    alphaInputType=Bool
    alphaBoolBlend {blendOption=Linear}
    alphaScaleBiasClamp {bInterpResult=true}
    boneToModify {boneName=dyn_l_fr_dressStrap_MSTR}
    lookAtLocation=<-4.3025 -2.6007 8.0333>
    lookAtTarget {boneReference {boneName=dyn_l_fr_dress_03}}
    lookAt_Axis {axis=<1 0>}
    lookUp_Axis {axis {x=0}}
   }
   n49 = Look At - Bone: dyn_r_fr_dressStrap_MSTR!(ComponentPose=@n48, bAlphaBoolEnabled:bool=@n47) -> (Pose) {
    alphaInputType=Bool
    alphaBoolBlend {blendOption=Linear}
    alphaScaleBiasClamp {bInterpResult=true}
    boneToModify {boneName=dyn_r_fr_dressStrap_MSTR}
    lookAtLocation=<-1.7847 7.8154 2.562>
    lookAtTarget {boneReference {boneName=dyn_r_fr_dress_03}}
    lookAt_Axis {axis=<1 0>}
    lookUp_Axis {axis {x=0}}
   }
   n50 = Get NewFrontEndEmoting() -> (NewFrontEndEmoting:bool)
   n51 = Look At - Bone: dyn_l_fr_dressStrap_MSTR(ComponentPose=@n49, bAlphaBoolEnabled=@n50) {
    alphaInputType=Bool
    alphaBoolBlend {blendOption=Linear}
    alphaScaleBiasClamp {bInterpResult=true}
    boneToModify {boneName=dyn_l_fr_dressStrap_MSTR}
    lookAtLocation=<-4.3025 -2.6007 8.0333>
    lookAtTarget {boneReference {boneName=dyn_l_fr_dress_03}}
    lookAt_Axis {axis=<1 0>}
    lookUp_Axis {axis {x=0}}
   }
   n52 = Look At - Bone: dyn_r_fr_dressStrap_MSTR(ComponentPose=@n51, bAlphaBoolEnabled=@n50) {
    alphaInputType=Bool
    alphaBoolBlend {blendOption=Linear}
    alphaScaleBiasClamp {bInterpResult=true}
    boneToModify {boneName=dyn_r_fr_dressStrap_MSTR}
    lookAtLocation=<-1.7847 7.8154 2.562>
    lookAtTarget {boneReference {boneName=dyn_r_fr_dress_03}}
    lookAt_Axis {axis=<1 0>}
    lookUp_Axis {axis {x=0}}
   }
   n53 = Get NotCrouching() -> (NotCrouching:bool)
   n54 = "Constraint - dyn_l_fr_dress_03 (RRRRRRR)"!(
    ComponentPose=@n52
    bAlphaBoolEnabled:bool=@n53
    ConstraintWeights_0=0.1177
    ConstraintWeights_1:float=0.0801
    ConstraintWeights_2:float=0.6755
    ConstraintWeights_3:float=0
    ConstraintWeights_4:float=0.1263
    ConstraintWeights_5:float=0.0002
    ConstraintWeights_6:float=0.0002
   ) -> (Pose) {
    alphaInputType=Bool
    alphaBoolBlend {blendInTime=0.1, blendOption=Linear, blendOutTime=0.1}
    boneToModify {boneName=dyn_l_fr_dress_03}
    constraintSetup [
     {transformType=Rotation, targetBone {boneName=pelvis}}
     {transformType=Rotation, targetBone {boneName=thigh_l}}
     {transformType=Rotation, targetBone {boneName=thigh_twist_01_l}}
     {transformType=Rotation, targetBone {boneName=deform_glute_base_l}}
     {transformType=Rotation, targetBone {boneName=deform_groin_base_l}}
     {transformType=Rotation, targetBone {boneName=thigh_r}}
     {transformType=Rotation, targetBone {boneName=deform_groin_base_r}}
    ]
    constraintWeights [1,1,1,1,1,1,1]
   }
   n55 = "Constraint - dyn_r_fr_dress_03 (RRRR)"!(
    ComponentPose=@n54
    bAlphaBoolEnabled:bool=@n53
    ConstraintWeights_0=0.1805
    ConstraintWeights_1:float=0.114
    ConstraintWeights_2:float=0.4882
    ConstraintWeights_3:float=0.2172
   ) -> (Pose) {
    alphaInputType=Bool
    alphaBoolBlend {blendInTime=0.1, blendOption=Linear, blendOutTime=0.1}
    boneToModify {boneName=dyn_r_fr_dress_03}
    constraintSetup [
     {transformType=Rotation, targetBone {boneName=pelvis}}
     {transformType=Rotation, targetBone {boneName=thigh_r}}
     {transformType=Rotation, targetBone {boneName=thigh_twist_01_r}}
     {transformType=Rotation, targetBone {boneName=deform_groin_base_r}}
    ]
    constraintWeights [1,1,1,1]
   }
   n56 = "Transform (Modify) Bone - Bone: dyn_l_fr_dress_03"(ComponentPose=@n55, bAlphaBoolEnabled=True, Translation="0,0,2", Rotation="7.664,-0.671,-1.203", Scale="1,1,1") {
    alphaInputType=Bool
    rotationMode=BMM_Additive
    rotationSpace=BCS_BoneSpace
    scaleMode=BMM_Additive
    scaleSpace=BCS_BoneSpace
    translationMode=BMM_Additive
    translationSpace=BCS_BoneSpace
    alphaBoolBlend {blendOption=Linear}
    boneToModify {boneName=dyn_l_fr_dress_03}
    rotation=<rot 7.664 -0.671 -1.203>
    translation {z=2}
   }
   n57 = Look At - Bone: dyn_r_fr_dress_03(ComponentPose=@n56, bAlphaBoolEnabled=True) {
    alphaInputType=Bool
    alphaBoolBlend {blendInTime=0.1, blendOption=Linear, blendOutTime=0.1}
    alphaScaleBiasClamp {bInterpResult=true}
    boneToModify {boneName=dyn_r_fr_dress_03}
    lookAtLocation=<-1.7847 7.8154 2.562>
    lookAtTarget {boneReference {boneName=calf_r}}
    lookAt_Axis {axis=<1 0>}
    lookUp_Axis {axis {x=0}}
   }
   n58 = Look At - Bone: dyn_l_fr_dress_03(ComponentPose=@n57, bAlphaBoolEnabled=True) {
    alphaInputType=Bool
    alphaBoolBlend {blendInTime=0.1, blendOption=Linear, blendOutTime=0.1}
    alphaScaleBiasClamp {bInterpResult=true}
    boneToModify {boneName=dyn_l_fr_dress_03}
    lookAtLocation=<-1.7847 7.8154 2.562>
    lookAtTarget {boneReference {boneName=calf_l}}
    lookAt_Axis {axis=<1 0>}
    lookUp_Axis {axis {x=0}}
   }
   n59 = "Transform (Modify) Bone - Bone: dyn_r_fr_dress_03"(ComponentPose=@n58, Alpha=Get thigh_rot_up_r() -> (thigh_rot_up_r:double), Translation="0,-5,5", Rotation="30,-20,0", Scale="0.8,1,1") {
    rotationMode=BMM_Additive
    rotationSpace=BCS_BoneSpace
    scaleMode=BMM_Additive
    scaleSpace=BCS_BoneSpace
    translationMode=BMM_Additive
    translationSpace=BCS_BoneSpace
    alphaBoolBlend {blendOption=Linear}
    boneToModify {boneName=dyn_r_fr_dress_03}
    rotation {pitch=30, yaw=-20}
    scale {x=0.8}
    translation {y=-5, z=5}
   }
   n60 = "Transform (Modify) Bone - Bone: dyn_l_fr_dress_03"!(ComponentPose=@n59, Alpha:float=Get thigh_rot_up_l() -> (thigh_rot_up_l:double), Translation="0,0,2", Rotation="20,20,0", Scale="0.8,1,1") -> (Pose) {
    rotationMode=BMM_Additive
    rotationSpace=BCS_BoneSpace
    scaleMode=BMM_Additive
    scaleSpace=BCS_BoneSpace
    translationMode=BMM_Additive
    translationSpace=BCS_BoneSpace
    alphaBoolBlend {blendOption=Linear}
    boneToModify {boneName=dyn_l_fr_dress_03}
    rotation {pitch=20, yaw=20}
    scale {x=0.8}
    translation {z=2}
   }
   n61 = "Transform (Modify) Bone - Bone: dyn_l_fr_dress_04"(ComponentPose=@n60, bAlphaBoolEnabled=True, Translation="0.359,1.923,1.236", Rotation="-22.133,6.570,-3.284", Scale="0.8,1,1") {
    alphaInputType=Bool
    rotationMode=BMM_Additive
    rotationSpace=BCS_BoneSpace
    scaleMode=BMM_Additive
    scaleSpace=BCS_BoneSpace
    translationMode=BMM_Additive
    translationSpace=BCS_BoneSpace
    alphaBoolBlend {blendOption=Linear}
    boneToModify {boneName=dyn_l_fr_dress_04}
    rotation=<rot -22.133 6.57 -3.284>
    scale {x=0.8}
    translation=<0.359 1.923 1.236>
   }
   n62 = "Transform (Modify) Bone - Bone: dyn_r_fr_dress_03"!(ComponentPose=@n61, bAlphaBoolEnabled:bool=Get NewJumping() -> (NewJumping:bool), Translation="0,0,2", Rotation="0,50,0", Scale="0.8,1,1") -> (Pose) {
    alphaInputType=Bool
    rotationMode=BMM_Additive
    rotationSpace=BCS_BoneSpace
    scaleMode=BMM_Additive
    scaleSpace=BCS_BoneSpace
    translationMode=BMM_Additive
    translationSpace=BCS_BoneSpace
    alphaBoolBlend {blendInTime=0.1, blendOption=Linear, blendOutTime=0.1}
    boneToModify {boneName=dyn_r_fr_dress_03}
    rotation {yaw=50}
    scale {x=0.8}
    translation {z=2}
   }
   n63 = "Constraint - dyn_l_bk_dressStrap_MSTR (PPPPPPPPPP)"!(
    ComponentPose=@n62
    Alpha=1
    ConstraintWeights_0=0
    ConstraintWeights_1:float=0
    ConstraintWeights_2:float=0.0106
    ConstraintWeights_3:float=0.1216
    ConstraintWeights_4:float=0.4662
    ConstraintWeights_5:float=0.0881
    ConstraintWeights_6:float=0.0002
    ConstraintWeights_7:float=0.0262
    ConstraintWeights_8:float=0.287
    ConstraintWeights_9:float=0
   ) -> (Pose) {
    alphaBoolBlend {blendOption=Linear}
    boneToModify {boneName=dyn_l_bk_dressStrap_MSTR}
    constraintSetup [
     {targetBone {boneName=spine_02}}
     {targetBone {boneName=spine_03}}
     {targetBone {boneName=spine_04}}
     {targetBone {boneName=spine_05}}
     {targetBone {boneName=clavicle_l}}
     {targetBone {boneName=deform_upperarm_bk_base_l}}
     {targetBone {boneName=deform_upperarm_fr_l}}
     {targetBone {boneName=deform_upperarm_up_base_l}}
     {targetBone {boneName=upperarm_twist_02_l}}
     {targetBone {boneName=deform_pec_l}}
    ]
    constraintWeights [1,1,1,1,1,1,1,1,1,1]
   }
   n64 = "Constraint - dyn_r_bk_dressStrap_MSTR (PPPPP)"!(
    ComponentPose=@n63
    Alpha=1
    ConstraintWeights_0=0.0356
    ConstraintWeights_1:float=0.4315
    ConstraintWeights_2:float=0.1709
    ConstraintWeights_3:float=0.1094
    ConstraintWeights_4:float=0.2525
   ) -> (Pose) {
    alphaBoolBlend {blendOption=Linear}
    boneToModify {boneName=dyn_r_bk_dressStrap_MSTR}
    constraintSetup [
     {targetBone {boneName=spine_05}}
     {targetBone {boneName=clavicle_r}}
     {targetBone {boneName=deform_upperarm_bk_r}}
     {targetBone {boneName=deform_upperarm_up_base_r}}
     {targetBone {boneName=upperarm_twist_02_r}}
    ]
    constraintWeights [1,1,1,1,1]
   }
   n65 = "Transform (Modify) Bone - Bone: dyn_r_bk_dress_03"(ComponentPose=@n64, bAlphaBoolEnabled=True, Translation="0,0,3", Rotation="23.490,-0.648,1.257", Scale="1,1,1") {
    alphaInputType=Bool
    rotationMode=BMM_Additive
    rotationSpace=BCS_BoneSpace
    scaleMode=BMM_Additive
    scaleSpace=BCS_BoneSpace
    translationMode=BMM_Additive
    translationSpace=BCS_BoneSpace
    alphaBoolBlend {blendOption=Linear}
    boneToModify {boneName=dyn_r_bk_dress_03}
    rotation=<rot 23.49 -0.648 1.257>
    translation {z=3}
   }
   n66 = Get NewNotFrontEnd()
   n67 = Look At - Bone: dyn_l_bk_dressStrap_MSTR!(ComponentPose=@n65, bAlphaBoolEnabled:bool=@n66) -> (Pose) {
    alphaInputType=Bool
    alphaBoolBlend {blendOption=Linear}
    alphaScaleBiasClamp {bInterpResult=true}
    boneToModify {boneName=dyn_l_bk_dressStrap_MSTR}
    lookAtLocation=<-0.3774 10.832 -9.6939>
    lookAtTarget {boneReference {boneName=dyn_l_bk_dress_03}}
    lookAt_Axis {axis=<1 0>}
    lookUp_Axis {axis {x=0}}
   }
   n68 = Look At - Bone: dyn_r_bk_dressStrap_MSTR!(ComponentPose=@n67, bAlphaBoolEnabled:bool=@n66) -> (Pose) {
    alphaInputType=Bool
    alphaBoolBlend {blendOption=Linear}
    alphaScaleBiasClamp {bInterpResult=true}
    boneToModify {boneName=dyn_r_bk_dressStrap_MSTR}
    lookAtLocation=<0.8936 1.3403 6.9686>
    lookAtTarget {boneReference {boneName=dyn_r_bk_dress_03}}
    lookAt_Axis {axis=<1 0>}
    lookUp_Axis {axis {x=0}}
   }
   n69 = Get NewFrontEndEmoting()
   n70 = Look At - Bone: dyn_l_bk_dressStrap_MSTR(ComponentPose=@n68, bAlphaBoolEnabled=@n69) {
    alphaInputType=Bool
    alphaBoolBlend {blendOption=Linear}
    alphaScaleBiasClamp {bInterpResult=true}
    boneToModify {boneName=dyn_l_bk_dressStrap_MSTR}
    lookAtLocation=<-0.3774 10.832 -9.6939>
    lookAtTarget {boneReference {boneName=dyn_l_bk_dress_03}}
    lookAt_Axis {axis=<1 0>}
    lookUp_Axis {axis {x=0}}
   }
   n71 = Look At - Bone: dyn_r_bk_dressStrap_MSTR(ComponentPose=@n70, bAlphaBoolEnabled=@n69) {
    alphaInputType=Bool
    alphaBoolBlend {blendOption=Linear}
    alphaScaleBiasClamp {bInterpResult=true}
    boneToModify {boneName=dyn_r_bk_dressStrap_MSTR}
    lookAtLocation=<0.8936 1.3403 6.9686>
    lookAtTarget {boneReference {boneName=dyn_r_bk_dress_03}}
    lookAt_Axis {axis=<1 0>}
    lookUp_Axis {axis {x=0}}
   }
   n72 = "Constraint - dyn_l_bk_dress_03 (RRRRRRRRR)"!(
    ComponentPose=@n71
    Alpha=Get thigh_rot_bk_l() -> (thigh_rot_bk_l:double)
    ConstraintWeights_0=0.0444
    ConstraintWeights_1:float=0.0006
    ConstraintWeights_2:float=0.2281
    ConstraintWeights_3:float=0.3887
    ConstraintWeights_4:float=0.3338
    ConstraintWeights_5:float=0.0018
    ConstraintWeights_6:float=0.0009
    ConstraintWeights_7:float=0.0017
    ConstraintWeights_8:float=0.0001
   ) -> (Pose) {
    alphaBoolBlend {blendOption=Linear}
    boneToModify {boneName=dyn_l_bk_dress_03}
    constraintSetup [
     {transformType=Rotation, targetBone {boneName=pelvis}}
     {transformType=Rotation, targetBone {boneName=spine_02}}
     {transformType=Rotation, targetBone {boneName=thigh_l}}
     {transformType=Rotation, targetBone {boneName=thigh_twist_01_l}}
     {transformType=Rotation, targetBone {boneName=deform_glute_base_l}}
     {transformType=Rotation, targetBone {boneName=deform_groin_base_l}}
     {transformType=Rotation, targetBone {boneName=thigh_r}}
     {transformType=Rotation, targetBone {boneName=thigh_twist_01_r}}
     {transformType=Rotation, targetBone {boneName=deform_groin_base_r}}
    ]
    constraintWeights [1,1,1,1,1,1,1,1,1]
   }
   n73 = "Constraint - dyn_r_bk_dress_03 (RRRRRRR)"!(
    ComponentPose=@n72
    Alpha=Get thigh_rot_bk_r() -> (thigh_rot_bk_r:double)
    ConstraintWeights_0=0.105
    ConstraintWeights_1:float=0.0002
    ConstraintWeights_2:float=0.0035
    ConstraintWeights_3:float=0.0012
    ConstraintWeights_4:float=0.3679
    ConstraintWeights_5:float=0.5217
    ConstraintWeights_6:float=0.0005
   ) -> (Pose) {
    alphaBoolBlend {blendOption=Linear}
    boneToModify {boneName=dyn_r_bk_dress_03}
    constraintSetup [
     {transformType=Rotation, targetBone {boneName=pelvis}}
     {transformType=Rotation, targetBone {boneName=thigh_l}}
     {transformType=Rotation, targetBone {boneName=thigh_twist_01_l}}
     {transformType=Rotation, targetBone {boneName=deform_glute_base_l}}
     {transformType=Rotation, targetBone {boneName=thigh_r}}
     {transformType=Rotation, targetBone {boneName=thigh_twist_01_r}}
     {transformType=Rotation, targetBone {boneName=deform_groin_base_r}}
    ]
    constraintWeights [1,1,1,1,1,1,1]
   }
   n74 = "Transform (Modify) Bone - Bone: dyn_l_bk_dressStrap_MSTR"(ComponentPose=@n73, bAlphaBoolEnabled=True, Translation="0,0,0", Rotation="50,0,0", Scale="0.5,1,1") {
    alphaInputType=Bool
    rotationMode=BMM_Additive
    rotationSpace=BCS_BoneSpace
    scaleMode=BMM_Additive
    scaleSpace=BCS_BoneSpace
    translationMode=BMM_Additive
    translationSpace=BCS_BoneSpace
    alphaBoolBlend {blendInTime=0.1, blendOption=Linear, blendOutTime=0.1}
    boneToModify {boneName=dyn_l_bk_dressStrap_MSTR}
    rotation {pitch=50}
    scale {x=0.5}
   }
   n75 = "Transform (Modify) Bone - Bone: dyn_r_bk_dressStrap_MSTR"(ComponentPose=@n74, bAlphaBoolEnabled=True, Translation="0,0,0", Rotation="10,0,0", Scale="0.5,1,1") {
    alphaInputType=Bool
    rotationMode=BMM_Additive
    rotationSpace=BCS_BoneSpace
    scaleMode=BMM_Additive
    scaleSpace=BCS_BoneSpace
    translationMode=BMM_Additive
    translationSpace=BCS_BoneSpace
    alphaBoolBlend {blendInTime=0.1, blendOption=Linear, blendOutTime=0.1}
    boneToModify {boneName=dyn_r_bk_dressStrap_MSTR}
    rotation {pitch=10}
    scale {x=0.5}
   }
   n76 = "Transform (Modify) Bone - Bone: dyn_l_fr_dressStrap_MSTR"!(ComponentPose=@n75, bAlphaBoolEnabled:bool=True, Translation="0,0,0", Rotation="0,0,0", Scale="0.5,1,1") -> (Pose) {
    alphaInputType=Bool
    rotationMode=BMM_Additive
    rotationSpace=BCS_BoneSpace
    scaleMode=BMM_Additive
    scaleSpace=BCS_BoneSpace
    translationMode=BMM_Additive
    translationSpace=BCS_BoneSpace
    alphaBoolBlend {blendInTime=0.1, blendOption=Linear, blendOutTime=0.1}
    boneToModify {boneName=dyn_l_fr_dressStrap_MSTR}
    scale {x=0.5}
   }
   n77 = "Transform (Modify) Bone - Bone: dyn_r_fr_dressStrap_MSTR"!(ComponentPose=@n76, bAlphaBoolEnabled:bool=True, Translation="0,0,0", Rotation="0,0,0", Scale="0.5,1,1") -> (Pose) {
    alphaInputType=Bool
    rotationMode=BMM_Additive
    rotationSpace=BCS_BoneSpace
    scaleMode=BMM_Additive
    scaleSpace=BCS_BoneSpace
    translationMode=BMM_Additive
    translationSpace=BCS_BoneSpace
    alphaBoolBlend {blendInTime=0.1, blendOption=Linear, blendOutTime=0.1}
    boneToModify {boneName=dyn_r_fr_dressStrap_MSTR}
    scale {x=0.5}
   }
   n78 = Component To Local(ComponentPose=@n77)
   n79 = Get bRigidBodyEnabled()
   n80 = Anim Dynamics - Bone: dyn_squid_base!(
    ComponentPose=Local To Component(LocalPose=Blend Poses by bool(BlendPose_0=@n41, BlendPose_1=@n78, bActiveValue=@n79) {blendType=Linear})
    Alpha=0.5
    GravityOverride:Vector="0,0,0"
    ExternalForce="0,0,0"
   ) -> (Pose) {
    angularDampingOverride=1
    bOverrideAngularDamping=true
    simulationSpace=BoneRelative
    alphaBoolBlend {blendOption=Linear}
    boundBone {boneName=dyn_squid_base}
    componentAppliedLinearAccClamp=<5000 5000 5000>
    componentLinearAccScale=<1.1 1.1 1.1>
    componentLinearVelScale=<1.1 1.1 1.1>
    physicsBodyDefinitions [{
      sphereCollisionRadius=0
      boundBone {boneName=dyn_squid_base}
      boxExtents=<5 10 10>
      constraintSetup {angularLimitsMax=<65 0 65>, angularLimitsMin=<-65 -65 -65>, angularTarget=<0 0 0>}
      localJointOffset=<1.5 -0.05 0.1>
     }]
    relativeSpaceBone {boneName=root}
   }
   n81 = Anim Dynamics - Bone: dyn_squid_chain_01!(ComponentPose=@n80, Alpha=1, GravityOverride:Vector="0,0,0", ExternalForce="0,0,0") -> (Pose) {
    angularDampingOverride=1
    bOverrideAngularDamping=true
    simulationSpace=BoneRelative
    alphaBoolBlend {blendOption=Linear}
    boundBone {boneName=dyn_squid_chain_01}
    componentAppliedLinearAccClamp=<5000 5000 5000>
    componentLinearAccScale=<1.1 1.1 1.1>
    componentLinearVelScale=<1.1 1.1 1.1>
    physicsBodyDefinitions [{
      sphereCollisionRadius=0
      boundBone {boneName=dyn_squid_chain_01}
      boxExtents=<5 10 10>
      constraintSetup {angularLimitsMax=<65 0 65>, angularLimitsMin=<-65 -65 -65>, angularTarget=<0 0 0>}
      localJointOffset=<1.5 -0.05 0.1>
     }]
    relativeSpaceBone {boneName=root}
   }
   n82 = Anim Dynamics - Bone: dyn_squid_tenticle_a_01!(ComponentPose=@n81, Alpha=1, GravityOverride:Vector="0,0,0", ExternalForce="0,0,0") -> (Pose) {
    angularDampingOverride=1
    bOverrideAngularDamping=true
    simulationSpace=BoneRelative
    alphaBoolBlend {blendOption=Linear}
    boundBone {boneName=dyn_squid_tenticle_a_01}
    componentAppliedLinearAccClamp=<5000 5000 5000>
    componentLinearAccScale=<1.1 1.1 1.1>
    componentLinearVelScale=<1.1 1.1 1.1>
    physicsBodyDefinitions [{
      sphereCollisionRadius=0
      boundBone {boneName=dyn_squid_tenticle_a_01}
      boxExtents=<5 10 10>
      constraintSetup {angularLimitsMax=<65 0 65>, angularLimitsMin=<-65 -65 -65>, angularTarget=<0 0 0>}
      localJointOffset=<1.5 -0.05 0.1>
     }]
    relativeSpaceBone {boneName=root}
   }
   n83 = Anim Dynamics - Bone: dyn_squid_tenticle_b_01!(ComponentPose=@n82, Alpha=1, GravityOverride:Vector="0,0,0", ExternalForce="0,0,0") -> (Pose) {
    angularDampingOverride=1
    bOverrideAngularDamping=true
    simulationSpace=BoneRelative
    alphaBoolBlend {blendOption=Linear}
    boundBone {boneName=dyn_squid_tenticle_b_01}
    componentAppliedLinearAccClamp=<5000 5000 5000>
    componentLinearAccScale=<1.1 1.1 1.1>
    componentLinearVelScale=<1.1 1.1 1.1>
    physicsBodyDefinitions [{
      sphereCollisionRadius=0
      boundBone {boneName=dyn_squid_tenticle_b_01}
      boxExtents=<5 10 10>
      constraintSetup {angularLimitsMax=<65 0 65>, angularLimitsMin=<-65 -65 -65>, angularTarget=<0 0 0>}
      localJointOffset=<1.5 -0.05 0.1>
     }]
    relativeSpaceBone {boneName=root}
   }
   n84 = Anim Dynamics - Bone: dyn_squid_chain_02!(ComponentPose=@n83, Alpha=1, GravityOverride:Vector="0,0,0", ExternalForce="0,0,0") -> (Pose) {
    angularDampingOverride=1
    bOverrideAngularDamping=true
    simulationSpace=BoneRelative
    alphaBoolBlend {blendOption=Linear}
    boundBone {boneName=dyn_squid_chain_02}
    componentAppliedLinearAccClamp=<5000 5000 5000>
    componentLinearAccScale=<1.1 1.1 1.1>
    componentLinearVelScale=<1.1 1.1 1.1>
    physicsBodyDefinitions [{
      sphereCollisionRadius=0
      boundBone {boneName=dyn_squid_chain_02}
      boxExtents=<5 10 10>
      constraintSetup {angularLimitsMax=<65 0 65>, angularLimitsMin=<-65 -65 -65>, angularTarget=<0 0 0>}
      localJointOffset=<1.5 -0.05 0.1>
     }]
    relativeSpaceBone {boneName=root}
   }
   n85 = Anim Dynamics - Bone: dyn_squid_tenticle_a_02!(ComponentPose=@n84, Alpha=1, GravityOverride:Vector="0,0,0", ExternalForce="0,0,0") -> (Pose) {
    angularDampingOverride=1
    bOverrideAngularDamping=true
    simulationSpace=BoneRelative
    alphaBoolBlend {blendOption=Linear}
    boundBone {boneName=dyn_squid_tenticle_a_02}
    componentAppliedLinearAccClamp=<5000 5000 5000>
    componentLinearAccScale=<1.1 1.1 1.1>
    componentLinearVelScale=<1.1 1.1 1.1>
    physicsBodyDefinitions [{
      sphereCollisionRadius=0
      boundBone {boneName=dyn_squid_tenticle_a_02}
      boxExtents=<5 10 10>
      constraintSetup {angularLimitsMax=<65 0 65>, angularLimitsMin=<-65 -65 -65>, angularTarget=<0 0 0>}
      localJointOffset=<1.5 -0.05 0.1>
     }]
    relativeSpaceBone {boneName=root}
   }
   n86 = Anim Dynamics - Bone: dyn_squid_tenticle_b_02!(ComponentPose=@n85, Alpha=1, GravityOverride:Vector="0,0,0", ExternalForce="0,0,0") -> (Pose) {
    angularDampingOverride=1
    bOverrideAngularDamping=true
    simulationSpace=BoneRelative
    alphaBoolBlend {blendOption=Linear}
    boundBone {boneName=dyn_squid_tenticle_b_02}
    componentAppliedLinearAccClamp=<5000 5000 5000>
    componentLinearAccScale=<1.1 1.1 1.1>
    componentLinearVelScale=<1.1 1.1 1.1>
    physicsBodyDefinitions [{
      sphereCollisionRadius=0
      boundBone {boneName=dyn_squid_tenticle_b_02}
      boxExtents=<5 10 10>
      constraintSetup {angularLimitsMax=<65 0 65>, angularLimitsMin=<-65 -65 -65>, angularTarget=<0 0 0>}
      localJointOffset=<1.5 -0.05 0.1>
     }]
    relativeSpaceBone {boneName=root}
   }
   n87 = "Constraint - dyn_shortSkirt_MSTR (PPPPPPPPP)"!(
    ComponentPose=@n86
    Alpha=1
    ConstraintWeights_0=0.106
    ConstraintWeights_1:float=0.1292
    ConstraintWeights_2:float=0.1972
    ConstraintWeights_3:float=0.3219
    ConstraintWeights_4:float=0.0002
    ConstraintWeights_5:float=0.0011
    ConstraintWeights_6:float=0.0091
    ConstraintWeights_7:float=0.0001
    ConstraintWeights_8:float=0.2351
   ) -> (Pose) {
    alphaBoolBlend {blendOption=Linear}
    boneToModify {boneName=dyn_shortSkirt_MSTR}
    constraintSetup [
     {targetBone {boneName=pelvis}}
     {targetBone {boneName=thigh_l}}
     {targetBone {boneName=thigh_twist_01_l}}
     {targetBone {boneName=deform_glute_base_l}}
     {targetBone {boneName=deform_groin_base_l}}
     {targetBone {boneName=thigh_r}}
     {targetBone {boneName=deform_glute_base_r}}
     {targetBone {boneName=thigh_twist_01_r}}
     {targetBone {boneName=dyn_shortSkirt}}
    ]
    constraintWeights [1,1,1,1,1,1,1,1,1]
   }
   n88 = Anim Dynamics - Bone: dyn_shortSkirt(ComponentPose=@n87, Alpha=1, ExternalForce="0,0,0") {
    angularDampingOverride=0.9999
    bGravityOverrideInSimSpace=true
    bOverrideAngularDamping=true
    simulationSpace=RootRelative
    alphaBoolBlend {blendOption=Linear}
    boundBone {boneName=dyn_shortSkirt}
    componentAppliedLinearAccClamp=<5000 5000 5000>
    componentLinearAccScale=<0.25 0.25 0.25>
    componentLinearVelScale=<0.25 0.25 0.25>
    physicsBodyDefinitions [{
      sphereCollisionRadius=0
      boundBone {boneName=dyn_shortSkirt}
      boxExtents=<10 20 20>
      constraintSetup {angularLimitsMax=<1 90 1>, angularLimitsMin=<-1 -90 -1>, angularTarget=<0 0 0>}
      localJointOffset=<4 0 0>
     }]
   }
   n89 = "Transform (Modify) Bone - Bone: dyn_shortSkirt_MSTR"(ComponentPose=@n88, Alpha=1, Translation="-5.758,-0.052,6.012", Rotation="0,0,0", Scale="1,1,1") {
    rotationMode=BMM_Additive
    rotationSpace=BCS_BoneSpace
    scaleMode=BMM_Additive
    scaleSpace=BCS_BoneSpace
    translationMode=BMM_Additive
    translationSpace=BCS_BoneSpace
    alphaBoolBlend {blendOption=Linear}
    boneToModify {boneName=dyn_shortSkirt_MSTR}
    translation=<-5.758 -0.052 6.012>
   }
   n90 = "Constraint - dyn_shortSkirt (P)"!(ComponentPose=@n89, bAlphaBoolEnabled:bool=True, ConstraintWeights_0=1) -> (Pose) {
    alphaInputType=Bool
    alphaBoolBlend {blendOption=Linear}
    boneToModify {boneName=dyn_shortSkirt}
    constraintSetup [{targetBone {boneName=thigh_twist_01_l}}]
    constraintWeights [1]
   }
   n91 = "Constraint - dyn_l_lace_MSTR (PPP)"!(ComponentPose=@n90, Alpha=1, ConstraintWeights_0=0.7713, ConstraintWeights_1:float=0.0792, ConstraintWeights_2:float=0.1495) -> (Pose) {
    alphaBoolBlend {blendOption=Linear}
    boneToModify {boneName=dyn_l_lace_MSTR}
    constraintSetup [{targetBone {boneName=calf_twist_01_l}}, {targetBone {boneName=calf_l}}, {targetBone {boneName=calf_twist_02_l}}]
    constraintWeights [1,1,1]
   }
   n92 = "Constraint - dyn_r_lace_MSTR (PPP)"!(ComponentPose=@n91, Alpha=1, ConstraintWeights_0=0.7713, ConstraintWeights_1:float=0.0792, ConstraintWeights_2:float=0.1495) -> (Pose) {
    alphaBoolBlend {blendOption=Linear}
    boneToModify {boneName=dyn_r_lace_MSTR}
    constraintSetup [{targetBone {boneName=calf_twist_01_r}}, {targetBone {boneName=calf_r}}, {targetBone {boneName=calf_twist_02_r}}]
    constraintWeights [1,1,1]
   }
   n93 = Anim Dynamics - Bone: dyn_l_loop_01!(ComponentPose=@n92, Alpha=1, GravityOverride:Vector="0,0,0", ExternalForce="0,0,0") -> (Pose) {
    angularDampingOverride=1
    angularSpringConstant=1500
    bAngularSpring=true
    bGravityOverrideInSimSpace=true
    bOverrideAngularDamping=true
    gravityScale=0.35
    simulationSpace=RootRelative
    alphaBoolBlend {blendOption=Linear}
    alphaScaleBias {scale=2}
    boundBone {boneName=dyn_l_loop_01}
    physicsBodyDefinitions [{
      sphereCollisionRadius=0
      boundBone {boneName=dyn_l_loop_01}
      boxExtents=<10 20 20>
      constraintSetup {angularLimitsMax=<0 0 5>, angularLimitsMin=<0 -20 -5>, angularTarget=<0 0 0>}
      localJointOffset=<10 0 0>
     }]
   }
   n94 = Anim Dynamics - Bone: dyn_l_loop_03!(ComponentPose=@n93, Alpha=1, GravityOverride:Vector="0,0,0", ExternalForce="0,0,0") -> (Pose) {
    angularDampingOverride=1
    angularSpringConstant=1500
    bAngularSpring=true
    bGravityOverrideInSimSpace=true
    bOverrideAngularDamping=true
    gravityScale=0.35
    simulationSpace=RootRelative
    alphaBoolBlend {blendOption=Linear}
    alphaScaleBias {scale=2}
    boundBone {boneName=dyn_l_loop_03}
    physicsBodyDefinitions [{
      sphereCollisionRadius=0
      boundBone {boneName=dyn_l_loop_03}
      boxExtents=<10 20 20>
      constraintSetup {angularLimitsMax=<0 0 5>, angularLimitsMin=<0 -20 -5>, angularTarget=<0 0 0>}
      localJointOffset=<10 0 0>
     }]
   }
   n95 = Anim Dynamics - Bone: dyn_l_lace_01!(ComponentPose=@n94, Alpha=1, GravityOverride:Vector="0,0,0", ExternalForce="0,0,0") -> (Pose) {
    angularDampingOverride=1
    angularSpringConstant=1500
    bAngularSpring=true
    bGravityOverrideInSimSpace=true
    bOverrideAngularDamping=true
    gravityScale=0.35
    simulationSpace=RootRelative
    alphaBoolBlend {blendOption=Linear}
    alphaScaleBias {scale=2}
    boundBone {boneName=dyn_l_lace_01}
    physicsBodyDefinitions [{
      sphereCollisionRadius=0
      boundBone {boneName=dyn_l_lace_01}
      boxExtents=<10 20 20>
      constraintSetup {angularLimitsMax=<0 0 5>, angularLimitsMin=<0 -20 -5>, angularTarget=<0 0 0>}
      localJointOffset=<10 0 0>
     }]
   }
   n96 = Anim Dynamics - Bone: dyn_l_lace_03!(ComponentPose=@n95, Alpha=1, GravityOverride:Vector="0,0,0", ExternalForce="0,0,0") -> (Pose) {
    angularDampingOverride=1
    angularSpringConstant=1500
    bAngularSpring=true
    bGravityOverrideInSimSpace=true
    bOverrideAngularDamping=true
    gravityScale=0.35
    simulationSpace=RootRelative
    alphaBoolBlend {blendOption=Linear}
    alphaScaleBias {scale=2}
    boundBone {boneName=dyn_l_lace_03}
    physicsBodyDefinitions [{
      sphereCollisionRadius=0
      boundBone {boneName=dyn_l_lace_03}
      boxExtents=<10 20 20>
      constraintSetup {angularLimitsMax=<0 0 5>, angularLimitsMin=<0 -20 -5>, angularTarget=<0 0 0>}
      localJointOffset=<10 0 0>
     }]
   }
   n97 = Anim Dynamics - Bone: dyn_l_loop_02!(ComponentPose=@n96, Alpha=1, GravityOverride:Vector="0,0,0", ExternalForce="0,0,0") -> (Pose) {
    angularDampingOverride=1
    angularSpringConstant=1500
    bAngularSpring=true
    bGravityOverrideInSimSpace=true
    bOverrideAngularDamping=true
    gravityScale=0.35
    simulationSpace=RootRelative
    alphaBoolBlend {blendOption=Linear}
    alphaScaleBias {scale=2}
    boundBone {boneName=dyn_l_loop_02}
    physicsBodyDefinitions [{
      sphereCollisionRadius=0
      boundBone {boneName=dyn_l_loop_02}
      boxExtents=<10 20 20>
      constraintSetup {angularLimitsMax=<0 0 5>, angularLimitsMin=<0 -20 -5>, angularTarget=<0 0 0>}
      localJointOffset=<10 0 0>
     }]
   }
   n98 = Anim Dynamics - Bone: dyn_l_loop_04!(ComponentPose=@n97, Alpha=1, GravityOverride:Vector="0,0,0", ExternalForce="0,0,0") -> (Pose) {
    angularDampingOverride=1
    angularSpringConstant=1500
    bAngularSpring=true
    bGravityOverrideInSimSpace=true
    bOverrideAngularDamping=true
    gravityScale=0.35
    simulationSpace=RootRelative
    alphaBoolBlend {blendOption=Linear}
    alphaScaleBias {scale=2}
    boundBone {boneName=dyn_l_loop_04}
    physicsBodyDefinitions [{
      sphereCollisionRadius=0
      boundBone {boneName=dyn_l_loop_04}
      boxExtents=<10 20 20>
      constraintSetup {angularLimitsMax=<0 0 5>, angularLimitsMin=<0 -20 -5>, angularTarget=<0 0 0>}
      localJointOffset=<10 0 0>
     }]
   }
   n99 = Anim Dynamics - Bone: dyn_l_lace_02!(ComponentPose=@n98, Alpha=1, GravityOverride:Vector="0,0,0", ExternalForce="0,0,0") -> (Pose) {
    angularDampingOverride=1
    angularSpringConstant=1500
    bAngularSpring=true
    bGravityOverrideInSimSpace=true
    bOverrideAngularDamping=true
    gravityScale=0.35
    simulationSpace=RootRelative
    alphaBoolBlend {blendOption=Linear}
    alphaScaleBias {scale=2}
    boundBone {boneName=dyn_l_lace_02}
    physicsBodyDefinitions [{
      sphereCollisionRadius=0
      boundBone {boneName=dyn_l_lace_02}
      boxExtents=<10 20 20>
      constraintSetup {angularLimitsMax=<0 0 5>, angularLimitsMin=<0 -20 -5>, angularTarget=<0 0 0>}
      localJointOffset=<10 0 0>
     }]
   }
   n100 = Anim Dynamics - Bone: dyn_l_lace_04!(ComponentPose=@n99, Alpha=1, GravityOverride:Vector="0,0,0", ExternalForce="0,0,0") -> (Pose) {
    angularDampingOverride=1
    angularSpringConstant=1500
    bAngularSpring=true
    bGravityOverrideInSimSpace=true
    bOverrideAngularDamping=true
    gravityScale=0.35
    simulationSpace=RootRelative
    alphaBoolBlend {blendOption=Linear}
    alphaScaleBias {scale=2}
    boundBone {boneName=dyn_l_lace_04}
    physicsBodyDefinitions [{
      sphereCollisionRadius=0
      boundBone {boneName=dyn_l_lace_04}
      boxExtents=<10 20 20>
      constraintSetup {angularLimitsMax=<0 0 5>, angularLimitsMin=<0 -20 -5>, angularTarget=<0 0 0>}
      localJointOffset=<10 0 0>
     }]
   }
   n101 = Anim Dynamics - Bone: dyn_r_loop_01!(ComponentPose=@n100, Alpha=1, GravityOverride:Vector="0,0,0", ExternalForce="0,0,0") -> (Pose) {
    angularDampingOverride=1
    angularSpringConstant=1500
    bAngularSpring=true
    bGravityOverrideInSimSpace=true
    bOverrideAngularDamping=true
    gravityScale=0.35
    simulationSpace=RootRelative
    alphaBoolBlend {blendOption=Linear}
    alphaScaleBias {scale=2}
    boundBone {boneName=dyn_r_loop_01}
    physicsBodyDefinitions [{
      sphereCollisionRadius=0
      boundBone {boneName=dyn_r_loop_01}
      boxExtents=<10 20 20>
      constraintSetup {angularLimitsMax=<0 0 5>, angularLimitsMin=<0 -20 -5>, angularTarget=<0 0 0>}
      localJointOffset=<10 0 0>
     }]
   }
   n102 = Anim Dynamics - Bone: dyn_r_loop_03!(ComponentPose=@n101, Alpha=1, GravityOverride:Vector="0,0,0", ExternalForce="0,0,0") -> (Pose) {
    angularDampingOverride=1
    angularSpringConstant=1500
    bAngularSpring=true
    bGravityOverrideInSimSpace=true
    bOverrideAngularDamping=true
    gravityScale=0.35
    simulationSpace=RootRelative
    alphaBoolBlend {blendOption=Linear}
    alphaScaleBias {scale=2}
    boundBone {boneName=dyn_r_loop_03}
    physicsBodyDefinitions [{
      sphereCollisionRadius=0
      boundBone {boneName=dyn_r_loop_03}
      boxExtents=<10 20 20>
      constraintSetup {angularLimitsMax=<0 0 5>, angularLimitsMin=<0 -20 -5>, angularTarget=<0 0 0>}
      localJointOffset=<10 0 0>
     }]
   }
   n103 = Anim Dynamics - Bone: dyn_r_lace_01!(ComponentPose=@n102, Alpha=1, GravityOverride:Vector="0,0,0", ExternalForce="0,0,0") -> (Pose) {
    angularDampingOverride=1
    angularSpringConstant=1500
    bAngularSpring=true
    bGravityOverrideInSimSpace=true
    bOverrideAngularDamping=true
    gravityScale=0.35
    simulationSpace=RootRelative
    alphaBoolBlend {blendOption=Linear}
    alphaScaleBias {scale=2}
    boundBone {boneName=dyn_r_lace_01}
    physicsBodyDefinitions [{
      sphereCollisionRadius=0
      boundBone {boneName=dyn_r_lace_01}
      boxExtents=<10 20 20>
      constraintSetup {angularLimitsMax=<0 0 5>, angularLimitsMin=<0 -20 -5>, angularTarget=<0 0 0>}
      localJointOffset=<10 0 0>
     }]
   }
   n104 = Anim Dynamics - Bone: dyn_r_lace_03!(ComponentPose=@n103, Alpha=1, GravityOverride:Vector="0,0,0", ExternalForce="0,0,0") -> (Pose) {
    angularDampingOverride=1
    angularSpringConstant=1500
    bAngularSpring=true
    bGravityOverrideInSimSpace=true
    bOverrideAngularDamping=true
    gravityScale=0.35
    simulationSpace=RootRelative
    alphaBoolBlend {blendOption=Linear}
    alphaScaleBias {scale=2}
    boundBone {boneName=dyn_r_lace_03}
    physicsBodyDefinitions [{
      sphereCollisionRadius=0
      boundBone {boneName=dyn_r_lace_03}
      boxExtents=<10 20 20>
      constraintSetup {angularLimitsMax=<0 0 5>, angularLimitsMin=<0 -20 -5>, angularTarget=<0 0 0>}
      localJointOffset=<10 0 0>
     }]
   }
   n105 = Anim Dynamics - Bone: dyn_r_loop_02!(ComponentPose=@n104, Alpha=1, GravityOverride:Vector="0,0,0", ExternalForce="0,0,0") -> (Pose) {
    angularDampingOverride=1
    angularSpringConstant=1500
    bAngularSpring=true
    bGravityOverrideInSimSpace=true
    bOverrideAngularDamping=true
    gravityScale=0.35
    simulationSpace=RootRelative
    alphaBoolBlend {blendOption=Linear}
    alphaScaleBias {scale=2}
    boundBone {boneName=dyn_r_loop_02}
    physicsBodyDefinitions [{
      sphereCollisionRadius=0
      boundBone {boneName=dyn_r_loop_02}
      boxExtents=<10 20 20>
      constraintSetup {angularLimitsMax=<0 0 5>, angularLimitsMin=<0 -20 -5>, angularTarget=<0 0 0>}
      localJointOffset=<10 0 0>
     }]
   }
   n106 = Anim Dynamics - Bone: dyn_r_loop_04!(ComponentPose=@n105, Alpha=1, GravityOverride:Vector="0,0,0", ExternalForce="0,0,0") -> (Pose) {
    angularDampingOverride=1
    angularSpringConstant=1500
    bAngularSpring=true
    bGravityOverrideInSimSpace=true
    bOverrideAngularDamping=true
    gravityScale=0.35
    simulationSpace=RootRelative
    alphaBoolBlend {blendOption=Linear}
    alphaScaleBias {scale=2}
    boundBone {boneName=dyn_r_loop_04}
    physicsBodyDefinitions [{
      sphereCollisionRadius=0
      boundBone {boneName=dyn_r_loop_04}
      boxExtents=<10 20 20>
      constraintSetup {angularLimitsMax=<0 0 5>, angularLimitsMin=<0 -20 -5>, angularTarget=<0 0 0>}
      localJointOffset=<10 0 0>
     }]
   }
   n107 = Anim Dynamics - Bone: dyn_r_lace_02!(ComponentPose=@n106, Alpha=1, GravityOverride:Vector="0,0,0", ExternalForce="0,0,0") -> (Pose) {
    angularDampingOverride=1
    angularSpringConstant=1500
    bAngularSpring=true
    bGravityOverrideInSimSpace=true
    bOverrideAngularDamping=true
    gravityScale=0.35
    simulationSpace=RootRelative
    alphaBoolBlend {blendOption=Linear}
    alphaScaleBias {scale=2}
    boundBone {boneName=dyn_r_lace_02}
    physicsBodyDefinitions [{
      sphereCollisionRadius=0
      boundBone {boneName=dyn_r_lace_02}
      boxExtents=<10 20 20>
      constraintSetup {angularLimitsMax=<0 0 5>, angularLimitsMin=<0 -20 -5>, angularTarget=<0 0 0>}
      localJointOffset=<10 0 0>
     }]
   }
   n108 = Anim Dynamics - Bone: dyn_r_lace_04!(ComponentPose=@n107, Alpha=1, GravityOverride:Vector="0,0,0", ExternalForce="0,0,0") -> (Pose) {
    angularDampingOverride=1
    angularSpringConstant=1500
    bAngularSpring=true
    bGravityOverrideInSimSpace=true
    bOverrideAngularDamping=true
    gravityScale=0.35
    simulationSpace=RootRelative
    alphaBoolBlend {blendOption=Linear}
    alphaScaleBias {scale=2}
    boundBone {boneName=dyn_r_lace_04}
    physicsBodyDefinitions [{
      sphereCollisionRadius=0
      boundBone {boneName=dyn_r_lace_04}
      boxExtents=<10 20 20>
      constraintSetup {angularLimitsMax=<0 0 5>, angularLimitsMin=<0 -20 -5>, angularTarget=<0 0 0>}
      localJointOffset=<10 0 0>
     }]
   }
   n109 = "Transform (Modify) Bone - Bone: dyn_backpack"!(ComponentPose=@n108, bAlphaBoolEnabled:bool=Get bIsSkyDiving(), Translation="0.000,-6,-1", Rotation="0,0,30", Scale="0.8,0.8,0.8") -> (Pose) {
    alphaInputType=Bool
    rotationMode=BMM_Additive
    translationMode=BMM_Additive
    alphaBoolBlend {blendInTime=0.1, blendOption=Linear, blendOutTime=0.1}
    boneToModify {boneName=dyn_backpack}
    rotation {roll=30}
    scale=<0.8 0.8 0.8>
    translation {y=-6, z=-1}
   }
   n110 = "Transform (Modify) Bone - Bone: dyn_sideSkirt"!(ComponentPose=@n109, bAlphaBoolEnabled:bool=Get CrouchWalking(), Translation="0,-1,0", Rotation="0,-10,0.000", Scale="1,1,1") -> (Pose) {
    alphaInputType=Bool
    rotationMode=BMM_Additive
    rotationSpace=BCS_BoneSpace
    scaleMode=BMM_Additive
    scaleSpace=BCS_BoneSpace
    translationMode=BMM_Additive
    translationSpace=BCS_BoneSpace
    alphaBoolBlend {blendInTime=0.1, blendOption=Linear, blendOutTime=0.1}
    boneToModify {boneName=dyn_sideSkirt}
    rotation {yaw=-10}
    translation {y=-1}
   }
   n111 = RigidBody!(ComponentPose=@n110, bAlphaBoolEnabled:bool, OverrideWorldGravity="0,0,-980", ExternalForce, ComponentLinearAccScale="0.4,0.4,0.5") -> (Pose) {
    alphaInputType=Bool
    bClampLinearTranslationLimitToRefPose=true
    bOverrideWorldGravity=true
    overridePhysicsAsset=/BRCosmetics/Characters/Player/Female/Medium/Bodies/F_MED_Gloom/Meshes/F_MED_Gloom_Physics
    simulationSpace=BaseBoneSpace
    alphaBoolBlend {blendOption=Linear}
    baseBoneRef {boneName=root}
    componentAppliedLinearAccClamp=<3000 3000 3000>
    componentLinearAccScale=<0.4 0.4 0.5>
    componentLinearVelScale=<0.3 0.3 0.4>
    overrideWorldGravity {z=-980}
    simSpaceSettings {
     maxAngularAcceleration=1000
     maxAngularVelocity=3
     maxLinearAcceleration=700
     maxLinearVelocity=500
     velocityScaleZ=0
     externalLinearDragV=<0.2 0.2>
    }
   }
   n112 = "Transform (Modify) Bone - Bone: dyn_frontSkirt"!(ComponentPose=@n111, bAlphaBoolEnabled:bool=True, Translation="0,0,1.5", Rotation="0,0,0", Scale="1,1,1") -> (Pose) {
    alphaInputType=Bool
    rotationMode=BMM_Additive
    rotationSpace=BCS_BoneSpace
    scaleMode=BMM_Additive
    scaleSpace=BCS_BoneSpace
    translationMode=BMM_Additive
    translationSpace=BCS_BoneSpace
    alphaBoolBlend {blendOption=Linear}
    boneToModify {boneName=dyn_frontSkirt}
    translation {z=1.5}
   }
   n113 = "Transform (Modify) Bone - Bone: dyn_sideSkirt"(ComponentPose=@n112, bAlphaBoolEnabled=True, Translation="0,-1,1", Rotation="0,0,0", Scale="1,1,1") {
    alphaInputType=Bool
    rotationMode=BMM_Additive
    rotationSpace=BCS_BoneSpace
    scaleMode=BMM_Additive
    scaleSpace=BCS_BoneSpace
    translationMode=BMM_Additive
    translationSpace=BCS_BoneSpace
    alphaBoolBlend {blendOption=Linear}
    boneToModify {boneName=dyn_sideSkirt}
    translation {y=-1, z=1}
   }
   n114 = "Transform (Modify) Bone - Bone: dyn_sideSkirt"!(ComponentPose=@n113, Alpha:float=1, Translation="0,-1,0", Rotation="-0.006,-3.793,0.000", Scale="1,1,1") -> (Pose) {
    rotationMode=BMM_Additive
    rotationSpace=BCS_BoneSpace
    scaleMode=BMM_Additive
    scaleSpace=BCS_BoneSpace
    translationMode=BMM_Additive
    translationSpace=BCS_BoneSpace
    alphaBoolBlend {blendOption=Linear}
    boneToModify {boneName=dyn_sideSkirt}
    rotation=<rot -0.006 -3.793 0>
    translation {y=-1}
   }
   n115 = "Constraint - dyn_r_fr_dress_03 (P)"!(ComponentPose=@n114, bAlphaBoolEnabled:bool=True, ConstraintWeights_0=1) -> (Pose) {
    alphaInputType=Bool
    alphaBoolBlend {blendInTime=0.1, blendOption=Linear, blendOutTime=0.3}
    boneToModify {boneName=dyn_r_fr_dress_03}
    constraintSetup [{targetBone {boneName=thigh_r}}]
    constraintWeights [1]
   }
   n116 = "Constraint - dyn_r_fr_dress_04 (P)"!(ComponentPose=@n115, bAlphaBoolEnabled:bool=True, ConstraintWeights_0=1) -> (Pose) {
    alphaInputType=Bool
    alphaBoolBlend {blendInTime=0.1, blendOption=Linear, blendOutTime=0.3}
    boneToModify {boneName=dyn_r_fr_dress_04}
    constraintSetup [{targetBone {boneName=thigh_r}}]
    constraintWeights [1]
   }
   n117 = "Constraint - dyn_r_fr_dress_05 (P)"!(ComponentPose=@n116, bAlphaBoolEnabled:bool=True, ConstraintWeights_0=1) -> (Pose) {
    alphaInputType=Bool
    alphaBoolBlend {blendInTime=0.1, blendOption=Linear, blendOutTime=0.3}
    boneToModify {boneName=dyn_r_fr_dress_05}
    constraintSetup [{targetBone {boneName=thigh_r}}]
    constraintWeights [1]
   }
   n118 = "Constraint - dyn_l_fr_dress_03 (P)"!(ComponentPose=@n117, bAlphaBoolEnabled:bool=True, ConstraintWeights_0=1) -> (Pose) {
    alphaInputType=Bool
    alphaBoolBlend {blendInTime=0.1, blendOption=Linear, blendOutTime=0.3}
    boneToModify {boneName=dyn_l_fr_dress_03}
    constraintSetup [{targetBone {boneName=thigh_l}}]
    constraintWeights [1]
   }
   n119 = "Constraint - dyn_l_fr_dress_04 (P)"!(ComponentPose=@n118, bAlphaBoolEnabled:bool=True, ConstraintWeights_0=1) -> (Pose) {
    alphaInputType=Bool
    alphaBoolBlend {blendInTime=0.1, blendOption=Linear, blendOutTime=0.3}
    boneToModify {boneName=dyn_l_fr_dress_04}
    constraintSetup [{targetBone {boneName=thigh_l}}]
    constraintWeights [1]
   }
   n120 = "Constraint - dyn_l_fr_dress_05 (P)"!(ComponentPose=@n119, bAlphaBoolEnabled:bool=True, ConstraintWeights_0=1) -> (Pose) {
    alphaInputType=Bool
    alphaBoolBlend {blendInTime=0.1, blendOption=Linear, blendOutTime=0.3}
    boneToModify {boneName=dyn_l_fr_dress_05}
    constraintSetup [{targetBone {boneName=thigh_l}}]
    constraintWeights [1]
   }
   n121 = Get SprintingNew() -> (SprintingNew:bool)
   n122 = "Constraint - dyn_r_fr_dress_03 (R)"!(ComponentPose=@n120, bAlphaBoolEnabled:bool=@n121, ConstraintWeights_0=1) -> (Pose) {
    alphaInputType=Bool
    alphaBoolBlend {blendInTime=0.1, blendOption=Linear, blendOutTime=0.3}
    alphaScaleBiasClamp {bInterpResult=true}
    boneToModify {boneName=dyn_r_fr_dress_03}
    constraintSetup [{transformType=Rotation, targetBone {boneName=thigh_r}}]
    constraintWeights [1]
   }
   n123 = "Constraint - dyn_r_fr_dress_04 (R)"!(ComponentPose=@n122, bAlphaBoolEnabled:bool=@n121, ConstraintWeights_0=1) -> (Pose) {
    alphaInputType=Bool
    alphaBoolBlend {blendInTime=0.1, blendOption=Linear, blendOutTime=0.3}
    alphaScaleBiasClamp {bInterpResult=true}
    boneToModify {boneName=dyn_r_fr_dress_04}
    constraintSetup [{transformType=Rotation, targetBone {boneName=thigh_r}}]
    constraintWeights [1]
   }
   n124 = "Constraint - dyn_r_fr_dress_05 (R)"!(ComponentPose=@n123, bAlphaBoolEnabled:bool=@n121, ConstraintWeights_0=1) -> (Pose) {
    alphaInputType=Bool
    alphaBoolBlend {blendInTime=0.1, blendOption=Linear, blendOutTime=0.3}
    alphaScaleBiasClamp {bInterpResult=true}
    boneToModify {boneName=dyn_r_fr_dress_05}
    constraintSetup [{transformType=Rotation, targetBone {boneName=thigh_r}}]
    constraintWeights [1]
   }
   n125 = "Constraint - dyn_l_fr_dress_03 (R)"!(ComponentPose=@n124, bAlphaBoolEnabled:bool=@n121, ConstraintWeights_0=1) -> (Pose) {
    alphaInputType=Bool
    alphaBoolBlend {blendInTime=0.1, blendOption=Linear, blendOutTime=0.3}
    alphaScaleBiasClamp {bInterpResult=true}
    boneToModify {boneName=dyn_l_fr_dress_03}
    constraintSetup [{transformType=Rotation, targetBone {boneName=thigh_l}}]
    constraintWeights [1]
   }
   n126 = "Constraint - dyn_l_fr_dress_04 (R)"!(ComponentPose=@n125, bAlphaBoolEnabled:bool=@n121, ConstraintWeights_0=1) -> (Pose) {
    alphaInputType=Bool
    alphaBoolBlend {blendInTime=0.1, blendOption=Linear, blendOutTime=0.3}
    alphaScaleBiasClamp {bInterpResult=true}
    boneToModify {boneName=dyn_l_fr_dress_04}
    constraintSetup [{transformType=Rotation, targetBone {boneName=thigh_l}}]
    constraintWeights [1]
   }
   n127 = "Constraint - dyn_l_fr_dress_05 (R)"!(ComponentPose=@n126, bAlphaBoolEnabled:bool=@n121, ConstraintWeights_0=1) -> (Pose) {
    alphaInputType=Bool
    alphaBoolBlend {blendInTime=0.1, blendOption=Linear, blendOutTime=0.3}
    alphaScaleBiasClamp {bInterpResult=true}
    boneToModify {boneName=dyn_l_fr_dress_05}
    constraintSetup [{transformType=Rotation, targetBone {boneName=thigh_l}}]
    constraintWeights [1]
   }
   Output Pose(Result=Component To Local(ComponentPose=@n127)) {layerGroup=None}
   n128 = Control Rig(Source=Copy Pose From Mesh() {bUseAttachedParent=true}) {
    alphaBoolBlend {blendOption=Linear}
    controlRigAssetReference {blueprintRigClass=&/F_MED_Gloom_DeformRig_Master_ControlRig_new.F_MED_Gloom_DeformRig_Master_ControlRig_new_C}
   }
   n129 = "Transform (Modify) Bone - Bone: dyn_root_slope"!(
    ComponentPose=Local To Component(LocalPose=@n128)
    bAlphaBoolEnabled:bool=NOT Boolean(A=Get bCanPlayCustomAnimations() -> (bCanPlayCustomAnimations:bool))
    Translation=Get slope_root_trans() -> (slope_root_trans:Vector)
    Rotation=Get slope_foot_rot() -> (slope_foot_rot:Rotator)
    Scale="1,1,1"
   ) -> (Pose) {
    alphaInputType=Bool
    rotationMode=BMM_Replace
    rotationSpace=BCS_WorldSpace
    translationMode=BMM_Replace
    translationSpace=BCS_WorldSpace
    alphaBoolBlend {blendOption=Linear}
    boneToModify {boneName=dyn_root_slope}
   }
   n130 = "Transform (Modify) Bone - Bone: dyn_root_simspace"(ComponentPose=@n129, Alpha=1, Translation="0,0,0", Rotation=Get root_space_rotation() -> (root_space_rotation:Rotator), Scale="1,1,1") {
    rotationMode=BMM_Replace
    rotationSpace=BCS_WorldSpace
    translationSpace=BCS_BoneSpace
    alphaBoolBlend {blendOption=Linear}
    alphaScaleBiasClamp {bInterpResult=true, interpSpeedDecreasing=2, interpSpeedIncreasing=2}
    boneToModify {boneName=dyn_root_simspace}
   }
   n131 = Look At - Bone: dyn_l_fr_dress_MSTR(ComponentPose=@n130, Alpha=1) {
    alphaBoolBlend {blendOption=Linear}
    boneToModify {boneName=dyn_l_fr_dress_MSTR}
    lookAtLocation=<0 10 -7>
    lookAtTarget {boneReference {boneName=pelvis}}
    lookAt_Axis {axis=<1 0>}
   }
   n132 = Look At - Bone: dyn_r_fr_dress_MSTR(ComponentPose=@n131, Alpha=1) {
    alphaBoolBlend {blendOption=Linear}
    boneToModify {boneName=dyn_r_fr_dress_MSTR}
    lookAtLocation=<0 10 7>
    lookAtTarget {boneReference {boneName=pelvis}}
    lookAt_Axis {axis=<1 0>}
   }
   n133 = Look At - Bone: dyn_l_fr_dressStrap_MSTR!(ComponentPose=@n132, Alpha:float=1) -> (Pose) {
    alphaBoolBlend {blendOption=Linear}
    boneToModify {boneName=dyn_l_fr_dressStrap_MSTR}
    lookAtLocation=<0 10 -7>
    lookAtTarget {boneReference {boneName=pelvis}}
    lookAt_Axis {axis=<1 0>}
   }
   n134 = Look At - Bone: dyn_r_fr_dressStrap_MSTR!(ComponentPose=@n133, Alpha:float=1) -> (Pose) {
    alphaBoolBlend {blendOption=Linear}
    boneToModify {boneName=dyn_r_fr_dressStrap_MSTR}
    lookAtLocation=<0 10 7>
    lookAtTarget {boneReference {boneName=pelvis}}
    lookAt_Axis {axis=<1 0>}
   }
   n135 = Look At - Bone: dyn_l_bk_dress_MSTR(ComponentPose=@n134, Alpha=1) {
    alphaBoolBlend {blendOption=Linear}
    boneToModify {boneName=dyn_l_bk_dress_MSTR}
    lookAtLocation=<0 -10 -7>
    lookAtTarget {boneReference {boneName=pelvis}}
    lookAt_Axis {axis=<1 0>}
   }
   n136 = Look At - Bone: dyn_r_bk_dress_MSTR(ComponentPose=@n135, Alpha=1) {
    alphaBoolBlend {blendOption=Linear}
    boneToModify {boneName=dyn_r_bk_dress_MSTR}
    lookAtLocation=<0 -10 7>
    lookAtTarget {boneReference {boneName=pelvis}}
    lookAt_Axis {axis=<1 0>}
   }
   n137 = Look At - Bone: dyn_l_bk_dressStrap_MSTR!(ComponentPose=@n136, Alpha:float=1) -> (Pose) {
    alphaBoolBlend {blendOption=Linear}
    boneToModify {boneName=dyn_l_bk_dressStrap_MSTR}
    lookAtLocation=<0 -10 -7>
    lookAtTarget {boneReference {boneName=pelvis}}
    lookAt_Axis {axis=<1 0>}
   }
   n138 = Look At - Bone: dyn_r_bk_dressStrap_MSTR!(ComponentPose=@n137, Alpha:float=1) -> (Pose) {
    alphaBoolBlend {blendOption=Linear}
    boneToModify {boneName=dyn_r_bk_dressStrap_MSTR}
    lookAtLocation=<0 -10 7>
    lookAtTarget {boneReference {boneName=pelvis}}
    lookAt_Axis {axis=<1 0>}
   }
   n139 = "Transform (Modify) Bone - Bone: dyn_l_fr_dress_MSTR"!(ComponentPose=@n138, bAlphaBoolEnabled:bool=@n2, Translation="0,0,0", Rotation="0.000,0.000,-30", Scale="1,1,1") -> (Pose) {
    alphaInputType=Bool
    rotationMode=BMM_Additive
    scaleMode=BMM_Additive
    translationMode=BMM_Additive
    alphaBoolBlend {blendInTime=0.2, blendOption=Linear, blendOutTime=0.2}
    boneToModify {boneName=dyn_l_fr_dress_MSTR}
    rotation {roll=-30}
   }
   n140 = "Transform (Modify) Bone - Bone: dyn_r_fr_dress_MSTR"!(ComponentPose=@n139, bAlphaBoolEnabled:bool=@n2, Translation="0,0,0", Rotation="0,0,-30", Scale="1,1,1") -> (Pose) {
    alphaInputType=Bool
    rotationMode=BMM_Additive
    scaleMode=BMM_Additive
    translationMode=BMM_Additive
    alphaBoolBlend {blendInTime=0.2, blendOption=Linear, blendOutTime=0.2}
    boneToModify {boneName=dyn_r_fr_dress_MSTR}
    rotation {roll=-30}
   }
   n141 = "Constraint - dyn_l_fr_dressStrap_MSTR (PPPPPPPP)"!(
    ComponentPose=@n140
    Alpha=1
    ConstraintWeights_0=0.1008
    ConstraintWeights_1:float=0.1005
    ConstraintWeights_2:float=0.1828
    ConstraintWeights_3:float=0.1902
    ConstraintWeights_4:float=0.0537
    ConstraintWeights_5:float=0.0086
    ConstraintWeights_6:float=0.2926
    ConstraintWeights_7:float=0.0708
   ) -> (Pose) {
    alphaBoolBlend {blendOption=Linear}
    boneToModify {boneName=dyn_l_fr_dressStrap_MSTR}
    constraintSetup [
     {targetBone {boneName=spine_04}}
     {targetBone {boneName=spine_05}}
     {targetBone {boneName=clavicle_l}}
     {targetBone {boneName=upperarm_l}}
     {targetBone {boneName=deform_upperarm_fr_l}}
     {targetBone {boneName=deform_upperarm_up_base_l}}
     {targetBone {boneName=upperarm_twist_02_l}}
     {targetBone {boneName=deform_pec_l}}
    ]
    constraintWeights [1,1,1,1,1,1,1,1]
   }
   n142 = "Constraint - dyn_r_fr_dressStrap_MSTR (PPPPPPPPPPPPPPPPPPPPPPPP)"!(
    ComponentPose=@n141
    Alpha=1
    ConstraintWeights_0=0
    ConstraintWeights_1:float=0
    ConstraintWeights_2:float=0
    ConstraintWeights_3:float=0.0001
    ConstraintWeights_4:float=0.1211
    ConstraintWeights_5:float=0.1287
    ConstraintWeights_6:float=0
    ConstraintWeights_7:float=0
    ConstraintWeights_8:float=0
    ConstraintWeights_9:float=0
    ConstraintWeights_10:float=0
    ConstraintWeights_11:float=0
    ConstraintWeights_12:float=0.2372
    ConstraintWeights_13:float=0.0766
    ConstraintWeights_14:float=0.0003
    ConstraintWeights_15:float=0.0003
    ConstraintWeights_16:float=0.1184
    ConstraintWeights_17:float=0.0097
    ConstraintWeights_18:float=0.2523
    ConstraintWeights_19:float=0
    ConstraintWeights_20:float=0.0511
    ConstraintWeights_21:float=0.0006
    ConstraintWeights_22:float=0.0036
    ConstraintWeights_23:float=0
   ) -> (Pose) {
    alphaBoolBlend {blendOption=Linear}
    boneToModify {boneName=dyn_r_fr_dressStrap_MSTR}
    constraintSetup [
     {targetBone {boneName=pelvis}}
     {targetBone {boneName=spine_01}}
     {targetBone {boneName=spine_02}}
     {targetBone {boneName=spine_03}}
     {targetBone {boneName=spine_04}}
     {targetBone {boneName=spine_05}}
     {targetBone {boneName=neck_01}}
     {targetBone {boneName=clavicle_l}}
     {targetBone {boneName=deform_upperarm_bk_base_l}}
     {targetBone {boneName=deform_upperarm_fr_l}}
     {targetBone {boneName=deform_upperarm_up_base_l}}
     {targetBone {boneName=upperarm_twist_02_l}}
     {targetBone {boneName=clavicle_r}}
     {targetBone {boneName=upperarm_r}}
     {targetBone {boneName=deform_elbow_in_base_r}}
     {targetBone {boneName=deform_upperarm_bk_r}}
     {targetBone {boneName=deform_upperarm_fr_base_r}}
     {targetBone {boneName=deform_upperarm_up_base_r}}
     {targetBone {boneName=upperarm_twist_02_r}}
     {targetBone {boneName=deform_pec_l}}
     {targetBone {boneName=deform_pec_base_r}}
     {targetBone {boneName=deform_upperarm_bk_base_r}}
     {targetBone {boneName=deform_pec_r}}
     {targetBone {boneName=simspace}}
    ]
    constraintWeights [1,1,1,1,1,1,1,1,1,1,1,1,1,1,1,1,1,1,1,1,1,1,1,1]
   }
   n143 = "Constraint - dyn_l_bk_dressStrap_MSTR (PPPPPPPP)"!(
    ComponentPose=@n142
    Alpha=1
    ConstraintWeights_0=0.0034
    ConstraintWeights_1:float=0.0963
    ConstraintWeights_2:float=0.4433
    ConstraintWeights_3:float=0.1532
    ConstraintWeights_4:float=0.0621
    ConstraintWeights_5:float=0.0014
    ConstraintWeights_6:float=0.0402
    ConstraintWeights_7:float=0.2002
   ) -> (Pose) {
    alphaBoolBlend {blendOption=Linear}
    boneToModify {boneName=dyn_l_bk_dressStrap_MSTR}
    constraintSetup [
     {targetBone {boneName=spine_04}}
     {targetBone {boneName=spine_05}}
     {targetBone {boneName=clavicle_l}}
     {targetBone {boneName=upperarm_l}}
     {targetBone {boneName=deform_upperarm_bk_base_l}}
     {targetBone {boneName=deform_upperarm_fr_l}}
     {targetBone {boneName=deform_upperarm_up_base_l}}
     {targetBone {boneName=upperarm_twist_02_l}}
    ]
    constraintWeights [1,1,1,1,1,1,1,1]
   }
   n144 = "Constraint - dyn_r_bk_dressStrap_MSTR (PPPPPPPPPPPPPPPPPPPPPPPP)"!(
    ComponentPose=@n143
    Alpha=1
    ConstraintWeights_0=0
    ConstraintWeights_1:float=0
    ConstraintWeights_2:float=0
    ConstraintWeights_3:float=0
    ConstraintWeights_4:float=0.0061
    ConstraintWeights_5:float=0.0733
    ConstraintWeights_6:float=0
    ConstraintWeights_7:float=0
    ConstraintWeights_8:float=0
    ConstraintWeights_9:float=0
    ConstraintWeights_10:float=0
    ConstraintWeights_11:float=0
    ConstraintWeights_12:float=0.438
    ConstraintWeights_13:float=0.0766
    ConstraintWeights_14:float=0.0002
    ConstraintWeights_15:float=0.1026
    ConstraintWeights_16:float=0.0026
    ConstraintWeights_17:float=0.0743
    ConstraintWeights_18:float=0.2167
    ConstraintWeights_19:float=0
    ConstraintWeights_20:float=0
    ConstraintWeights_21:float=0.0095
    ConstraintWeights_22:float=0.0001
    ConstraintWeights_23:float=0
   ) -> (Pose) {
    alphaBoolBlend {blendOption=Linear}
    boneToModify {boneName=dyn_r_bk_dressStrap_MSTR}
    constraintSetup [
     {targetBone {boneName=pelvis}}
     {targetBone {boneName=spine_01}}
     {targetBone {boneName=spine_02}}
     {targetBone {boneName=spine_03}}
     {targetBone {boneName=spine_04}}
     {targetBone {boneName=spine_05}}
     {targetBone {boneName=neck_01}}
     {targetBone {boneName=clavicle_l}}
     {targetBone {boneName=deform_upperarm_bk_base_l}}
     {targetBone {boneName=deform_upperarm_fr_l}}
     {targetBone {boneName=deform_upperarm_up_base_l}}
     {targetBone {boneName=upperarm_twist_02_l}}
     {targetBone {boneName=clavicle_r}}
     {targetBone {boneName=upperarm_r}}
     {targetBone {boneName=deform_elbow_in_base_r}}
     {targetBone {boneName=deform_upperarm_bk_r}}
     {targetBone {boneName=deform_upperarm_fr_base_r}}
     {targetBone {boneName=deform_upperarm_up_base_r}}
     {targetBone {boneName=upperarm_twist_02_r}}
     {targetBone {boneName=deform_pec_l}}
     {targetBone {boneName=deform_pec_base_r}}
     {targetBone {boneName=deform_upperarm_bk_base_r}}
     {targetBone {boneName=deform_pec_r}}
     {targetBone {boneName=simspace}}
    ]
    constraintWeights [1,1,1,1,1,1,1,1,1,1,1,1,1,1,1,1,1,1,1,1,1,1,1,1]
   }
   Save cached pose 'PreSim'(Pose=Component To Local(ComponentPose=@n144)) {cacheName=PreSim}
  ]}
 EventGraph {nodes [
   Event BlueprintUpdateAnimation()
   > n1 = Sequence!(execute) -> (
    then_0
    then_1
    then_2:Exec
    then_3:Exec
    then_4:Exec
    then_5:Exec
    then_6:Exec
    then_7:Exec
    then_8:Exec
    then_9:Exec
    then_10:Exec
    then_11:Exec
    then_12:Exec
    then_13:Exec
    then_14:Exec
    then_15:Exec
   )
   n2 = Gravity Override State Function DEPRECATED(
    execute=@n1.then_1
    CharacterPartAnimInstance=Self-Reference()
    GameTimeInSeconds=GetGameTimeInSeconds()
    GravityOverrideParams=Get Gravity Override Params()
   )
   > Set GravityOverride(GravityOverride=@n2)
   n3 = Branch(execute=@n1.then_10, Condition=Get bIsInFrontEnd())
   > Set BwdMovement(BwdMovement:double=0.5) -> (Output_Get:double)
   n4 = PlayerMovement(execute=@n3.else, AnimInstance=Self-Reference(), XDotProductCoefficient=1, YDotProductCoefficient=1, ZDotProductCoefficient=1)
   > n5 = Branch(Condition=AND Boolean(A=Get bIsTargeting(), B=Get bIsCrouching()))
   n6 = MapRangeClamped(
    Value=float + float(A="Absolute (Float)"(A=@n4), B=MapRangeClamped(Value=@n4.YMovement, InRangeA=-1, InRangeB=1, OutRangeA=0.3)) {is=title K2Node_CommutativeAssociativeBinaryOperator}
    InRangeA=0
    InRangeB=1.65
   )
   > Set BwdMovement(BwdMovement=float * float(A=@n6, B=0.25) {is=title K2Node_PromotableOperator})
   Set BwdMovement(execute=@n5.else, BwdMovement=@n6)
   n7 = Branch(execute=@n1.then_11, Condition=Get bIsSkyDiving())
   > n8 = Branch(Condition="float > float"(A=Get PawnVelocityZ(), B=50) {is=title K2Node_CallFunction})
   > Set BwdMovement(BwdMovement=float * float(A=Get BwdMovement() -> (BwdMovement:double), B=0) {is=title K2Node_CommutativeAssociativeBinaryOperator})
   Set BwdMovement(execute=[@n7.else, @n8.else], BwdMovement=float * float(A=Get BwdMovement(), B=1))
   n9 = Get bIsSkyDiving()
   n10 = Get bIsTargeting()
   n11 = Get bIsPlayingMeleeAnim()
   Set ADS Hair Velocity(execute=@n1.then_2, ADS Hair Velocity:Vector=SelectVector(A="0.15,0.15,0.25", B="0.5,0.5,0.6", bPickA=OR Boolean(A=@n10, B=@n11, C=@n9))) -> (Output_Get:Vector)
   n12 = Get bIsTargeting()
   n13 = Get bIsSkyDiving()
   n14 = Get bIsPlayingMeleeAnim()
   > Set FrontDress Velocity(FrontDress Velocity:Vector=SelectVector(A="0.15,0.15,0.25", B="0.3,0.3,0.45", bPickA=OR Boolean(A=@n12, B=@n14, C=@n13))) -> (Output_Get:Vector)
   n15 = GetOwnerSkeletalMeshComponent()
   n16 = BreakRotator(InRot=BreakTransform(InTransform=GetDeltaTransformFromRefPose(self=@n15, BoneName=thigh_r, BaseName=pelvis)).Rotation)
   Set thigh_rot_up_r(execute=@n1.then_8, thigh_rot_up_r:double=MapRangeClamped(Value=@n16.Yaw, InRangeA=-10, InRangeB=-100)) -> (Output_Get:double)
   > Set thigh_rot_inr_r(thigh_rot_inr_r:double=MapRangeClamped(Value=@n16.Pitch, InRangeA=0, InRangeB=-10, OutRangeA=1, OutRangeB=0)) -> (Output_Get:double)
   > Set thigh_rot_otr_r(thigh_rot_otr_r:double=MapRangeClamped(Value=@n16.Pitch, InRangeA=20, InRangeB=50)) -> (Output_Get:double)
   > Set thigh_rot_bk_r(thigh_rot_bk_r:double=MapRangeClamped(Value=@n16.Yaw, InRangeA=-31, InRangeB=31)) -> (Output_Get:double)
   n17 = GetOwnerSkeletalMeshComponent()
   n18 = BreakRotator(InRot=BreakTransform(InTransform=GetDeltaTransformFromRefPose(self=@n17, BoneName=thigh_l, BaseName=pelvis)).Rotation)
   > Set thigh_rot_up_l(thigh_rot_up_l:double=MapRangeClamped(Value=@n18.Yaw, InRangeA=-10, InRangeB=-100)) -> (Output_Get:double)
   > Set thigh_rot_inr_l(thigh_rot_inr_l:double=MapRangeClamped(Value=@n18.Pitch, InRangeA=0, InRangeB=-10, OutRangeA=1, OutRangeB=0)) -> (Output_Get:double)
   > Set thigh_rot_otr_l(thigh_rot_otr_l:double=MapRangeClamped(Value=@n18.Pitch, InRangeA=20, InRangeB=50)) -> (Output_Get:double)
   > Set thigh_rot_bk_l(thigh_rot_bk_l:double=MapRangeClamped(Value=@n18.Yaw, InRangeA=-31, InRangeB=31)) -> (Output_Get:double)
   n19 = Branch(execute=@n1.then_13, Condition=NOT Boolean(A=Get bIsPlayingEmote()))
   > Set NewNotEmoting(NewNotEmoting:bool=true) -> (Output_Get:bool)
   Set NewNotEmoting(execute=@n19.else, NewNotEmoting=false)
   n20 = Branch(execute=@n1.then_14, Condition=NOT Boolean(A=Get bIsInFrontEnd()))
   > Set NewNotFrontEnd(NewNotFrontEnd:bool=true) -> (Output_Get:bool)
   Set NewNotFrontEnd(execute=@n20.else, NewNotFrontEnd=false)
   n21 = Branch(execute=@n1.then_15, Condition=AND Boolean(A=Get bIsInFrontEnd(), B=Get bIsPlayingEmote()))
   > Set NewFrontEndEmoting(NewFrontEndEmoting:bool=true) -> (Output_Get:bool)
   Set NewFrontEndEmoting(execute=@n21.else, NewFrontEndEmoting=false)
   n22 = Break Fort Anim Input Slope Info!(FortAnimInput_SlopeInfo:"Fort Anim Input Slope Info Structure (by ref)"=Get SlopeInfoInput() -> (SlopeInfoInput:Fort Anim Input Slope Info Structure)) -> (RootSlopeTranslation:Vector=~, LeftFootSlopeTranslation:Vector=~, RightFootSlopeTranslation:Vector=~, SlopeRotation:Rotator=~) {is=title K2Node_BreakStruct}
   Set slope_root_trans(execute=@n1.then_4, slope_root_trans:Vector=@n22) -> (Output_Get:Vector)
   > Set slope_foot_rot(slope_foot_rot:Rotator=@n22.SlopeRotation) -> (Output_Get:Rotator)
   n23 = Branch(execute=@n1.then_9, Condition=NOT Boolean(A=Get bIsCrouching()))
   > Set NotCrouching(NotCrouching:bool=true) -> (Output_Get:bool)
   Set NotCrouching(execute=@n23.else, NotCrouching=false)
   n24 = Branch(execute=@n1.then_7, Condition=Get bIsSkyDiving())
   > Set SkydivingNew(SkydivingNew:bool=true) -> (Output_Get:bool)
   Set SkydivingNew(execute=@n24.else, SkydivingNew=false)
   n25 = Branch(execute=@n1.then_5, Condition=AND Boolean(A=Get bIsAccelerating2D(), B=Get bIsCrouching()))
   > Set CrouchWalking(CrouchWalking:bool=true) -> (Output_Get:bool)
   Set CrouchWalking(execute=@n25.else, CrouchWalking=false)
   n26 = Branch(execute=@n1.then_6, Condition=Get bIsSprinting())
   > Set SprintingNew(SprintingNew:bool=true) -> (Output_Get:bool)
   > Set Sprinting_Float(Sprinting_Float:double=0.5) -> (Output_Get:double)
   Set SprintingNew(execute=@n26.else, SprintingNew=false)
   > Set Sprinting_Float(Sprinting_Float=0)
   Set thigh_rot_r(
    execute=@n1.then_0
    thigh_rot_r=GetAlphaFromReferenceTransformDelta(SkeletalMeshComponent=Get Skel_Mesh(), BoneName=thigh_r, BaseName=pelvis, InputRange=(X=0,Y=-90), TwistAxis="0,0,1")
   )
   > Set thigh_rot_l(thigh_rot_l=GetAlphaFromReferenceTransformDelta(SkeletalMeshComponent=Get Skel_Mesh(), BoneName=thigh_l, BaseName=pelvis, InputRange=(X=-20,Y=-101), TwistAxis="0,0,1"))
   n27 = ComplexSimSpaceFunction DEPRECATED!(
    execute=@n1.then_3
    CharacterPartAnimInstance:Custom Character Part Anim Instance Object Reference=Self-Reference()
    InterpSpeedMax:double=9
    PawnSpeedMax:double=20
    RInterpErrorTolerance:double=1
    DeltaTime:double=0
    DebugDraw:bool=false
    sim_space_rotation_in:Rotator=Get sim_space_rotation() -> (sim_space_rotation:Rotator)
    inComponentSpaceIn:bool=Get InComponentSpace() -> (InComponentSpace:bool)
    inComponentSpaceLastFrameIn:bool=Get InComponentSpaceLastFrame() -> (InComponentSpaceLastFrame:bool)
    lockedOnSimSpaceIn:bool=Get LockedOnSimSpace() -> (LockedOnSimSpace:bool)
    InterpSpeedIn:double=Get InterpSpeed() -> (InterpSpeed:double)
   ) -> (
    then
    sim_space_rotation:Rotator
    inComponentSpaceOut:bool
    inComponentSpaceLastFrameOut:bool
    lockedOnSimSpaceOut:bool
    InterpSpeedOut:double
   )
   > Set sim_space_rotation(sim_space_rotation:Rotator=@n27) -> (Output_Get:Rotator)
   > Set InComponentSpace(InComponentSpace:bool=@n27.inComponentSpaceOut) -> (Output_Get:bool)
   > Set InComponentSpaceLastFrame(InComponentSpaceLastFrame:bool=@n27.inComponentSpaceLastFrameOut) -> (Output_Get:bool)
   > Set LockedOnSimSpace(LockedOnSimSpace:bool=@n27.lockedOnSimSpaceOut) -> (Output_Get:bool)
   > Set InterpSpeed(InterpSpeed:double=@n27.InterpSpeedOut) -> (Output_Get:double)
   n28 = Branch(execute=@n1.then_12, Condition=OR Boolean!(A=Get bIsJumping(), B=Get bIsFalling()) -> (ReturnValue))
   > Set NewJumping(NewJumping:bool=true) -> (Output_Get:bool)
   Set NewJumping(execute=@n28.else, NewJumping=false)
   Event BlueprintInitializeAnimation()
   > Set Skel_Mesh(Skel_Mesh=GetOwnerSkeletalMeshComponent())
   Event OnBeginSkydiving()
   > Set SkydivingAlpha(SkydivingAlpha:Rotator="0,-65,0") -> (Output_Get:Rotator)
   Event OnEndSkydiving()
   > Set SkydivingAlpha(SkydivingAlpha="0,-10,0")
  ]}
]
variables [
 ADS Hair Velocity {type=Vector}
 BwdMovement {type=number}
 CrouchWalking {type=boolean}
 FrontDress Velocity {type=Vector}
 Gravity Override Params {
  type=GravityOverrideParamsStruct
  default {
   skydive_DivingDown {windAmplitude=<10 30 10>}
   skydive_FloatingDown {gravityOverride=<0 -700 0>, windAmplitude=<100 -200 200>, windFrequency=<0.5 2 3>}
   skydive_GoingUp {gravityOverride=<0 -700 -100>, windAmplitude=<100 -200 200>, windFrequency=<0.5 2 3>}
  }
 }
 GravityOverride {type=Vector}
 InComponentSpace {type=boolean}
 InComponentSpaceLastFrame {type=boolean}
 InterpSpeed {type=number}
 LockedOnSimSpace {type=boolean}
 NewEmoting {type=boolean}
 NewFrontEndEmoting {type=boolean}
 NewJumping {type=boolean}
 NewNotEmoting {type=boolean}
 NewNotFrontEnd {type=boolean}
 NotCrouching {type=boolean}
 root_space_rotation {type=Rotator}
 sim_space_rotation {type=Rotator}
 Skel_Mesh {type=SkeletalMeshComponent}
 SkydivingAlpha {type=Rotator}
 SkydivingNew {type=boolean}
 slope_foot_rot {type=Rotator}
 slope_root_trans {type=Vector}
 Sprinting_Float {type=number}
 SprintingNew {type=boolean}
 thigh_rot_bk_l {type=number}
 thigh_rot_bk_r {type=number}
 thigh_rot_inr_l {type=number}
 thigh_rot_inr_r {type=number}
 thigh_rot_l {type=number}
 thigh_rot_otr_l {type=number}
 thigh_rot_otr_r {type=number}
 thigh_rot_r {type=number}
 thigh_rot_up_l {type=number}
 thigh_rot_up_r {type=number}
]
